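# Translucent UI material for a sidebar island tile, showing a thumbnail with loading and error states.
# The thumbnail is scaled to fill and zoomable, with an optional manual media size so image and video match, and is darkened by an overlay opacity.
# Floored IsLoading and IsError scalars act as flags alongside an animated gradient loading mask and an MSDF image-error icon.
# A rounded-corner mask shapes the final opacity.
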Material /Sidebar/Materials/M_Sidebar_IslandTile
graph {
 properties {bAutomaticallySetUsageInEditor=false, blendMode=BLEND_Translucent, materialDomain=MD_UI, usedShadingModels=Default Lit}
 nodes [
  LinearInterpolate(
   A=NamedRerouteUsage()
   B=LinearInterpolate(B=Constant3Vector() {constant=<rgba 0.3864 0.3763 0.4793 1>}, Alpha=NamedRerouteUsage())
   Alpha=NamedRerouteUsage()
  )
  n1 = TextureObjectParameter() {
   group=1 - Thumbnail
   parameterName=Thumbnail
   samplerType=SAMPLERTYPE_Masks
   sortPriority=0
   texture=/Game/Global/Textures/Default/Blanks/T_MidgreyMask
  }
  n2 = TextureSample(
   UVs=MaterialFunctionCall(UVs=NamedRerouteUsage(), Texture Scale=ScalarParameter() {defaultValue=1, group=1 - Thumbnail, parameterName=Thumbnail Zoom, sortPriority=2}) {materialFunction=&ScaleUVsByCenter}.UVs
   Tex=NamedRerouteDeclaration(None=@n1) {name=Thumbnail, nodeColor=<rgba 0 0 0 1>}
  )
  n3 = ComponentMask(None=@n2.RGBA) {b=true, g=true, r=true}
  n4 = Constant3Vector() {desc=Darken Color, constant=<rgba 0 0 0 1>}
  n5 = Multiply(
   A=ScalarParameter() {group=1 - Thumbnail, parameterName=Overlay Opacity, sliderMax=1, sortPriority=3}
   B=Constant() {desc=Darken Maximum Amount, r=0.7}
  )
  NamedRerouteDeclaration(None=LinearInterpolate(A=LinearInterpolate(A=@n3, B=@n4, Alpha=@n5), B=Constant3Vector() {constant=<rgba 0.3864 0.3763 0.4793 1>}, Alpha=NamedRerouteUsage())) {name=Thumbnail Color, nodeColor=<rgba 0 0 0 1>}
  NamedRerouteDeclaration(None=MaterialFunctionCall() {materialFunction=&GetUserInterfaceUV}.Normalized UV) {name=Normalized UV, nodeColor=<rgba 0 0 0 1>}
  n6 = AppendVector(
   A=ScalarParameter() {defaultValue=16, group=1 - Thumbnail, parameterName=Media Width}
   B=ScalarParameter() {defaultValue=9, group=1 - Thumbnail, parameterName=Media Height, sortPriority=33}
  ) {desc=HACK - Used to send the texture size manually to have both image and video to be the same size}
  n7 = TextureProperty(Tex=NamedRerouteUsage())
  NamedRerouteDeclaration(None=MaterialFunctionCall(
    UV=NamedRerouteUsage()
    TextureSize=StaticSwitchParameter(True=@n6, False=@n7) {group=1 - Thumbnail, parameterName=Use Fixed Size, sortPriority=31}
   ) {materialFunction=&MF_UI_UVScaling}.Scaled To Fill) {name=Scale to Fill UV, nodeColor=<rgba 0 0 0 1>}
  NamedRerouteDeclaration(None=ComponentMask(None=@n2.RGBA) {a=true}) {name=Thumbnail Alpha, nodeColor=<rgba 0 0 0 1>}
  n8 = Constant() {r=2}
  n9 = Constant() {r=0.125}
  n10 = MaterialFunctionCall(GradientEnd=MaterialFunctionCall("Value (0-1)"=MaterialFunctionCall!(Animation Duration=@n8, Delay=@n9, Reset to Default) {materialFunction=/FortUILibrary/MaterialFunctions/Time/MF_UI_AnimationDelay}.Result) {materialFunction=&MF_UI_EaseCurves}.EaseInOutSine) {materialFunction=&MF_UI_LinearGradient}
  n11 = MaterialFunctionCall!(A=@n10.Result, B=Constant() {r=1}) {materialFunction=/FortUILibrary/MaterialFunctions/Basics/MF_UI_BooleanOperations}
  NamedRerouteDeclaration(None=Multiply(A=@n10.Result, B=OneMinus(None=@n11.IsEqual))) {name=Loading Mask, nodeColor=<rgba 0 0 0 1>}
  NamedRerouteDeclaration(None=Floor(None=ScalarParameter() {parameterName=IsLoading, sliderMax=1})) {name=Is Loading, nodeColor=<rgba 0.8516 1 0 1>}
  NamedRerouteDeclaration(None=Floor(None=ScalarParameter() {parameterName=IsError, sliderMax=1})) {name=Is Error, nodeColor=<rgba 1 0.7266 0 1>}
  n12 = NamedRerouteUsage()
  n13 = Constant2Vector() {g=0.75, r=0.75}
  n14 = MaterialFunctionCall!(MSDF RGB=TextureSample(
    UVs=MaterialFunctionCall(UVs=@n12, Texture Scale=@n13) {materialFunction=&ScaleUVsByCenter}.UVs
    Tex=TextureObject() {samplerType=SAMPLERTYPE_LinearColor, texture=/FortUILibrary/Textures/Icons/T_UI_Icon_UIKit_ImageError_MSDF}
   ) {samplerSource=SSM_Clamp_WorldGroupSettings}.RGB) {materialFunction=/FortUILibrary/MaterialFunctions/Shapes/Modifiers/MF_UI_MSDFtoSDF}
  n15 = Constant() {r=0.5}
  n16 = Constant() {r=0.025}
  NamedRerouteDeclaration(None=LinearInterpolate(Alpha=MaterialFunctionCall(DistanceField=@n14.SDF, Threshold=@n15, Smoothing=@n16) {materialFunction=&MF_UI_AntiAliasing}.Outer) {constA=0.18}) {name=Empty Thumbnail Mask, nodeColor=<rgba 0 0 0 1>}
  n17 = VectorParameter() {parameterName=Corner Radius, defaultValue=<rgba 12 12 12 12>}
  n18 = AppendVector(A=@n17.RGB, B=@n17.RGB)
  n19 = ScalarParameter() {defaultValue=1, parameterName=RadiusInPixels}
  n20 = LinearInterpolate(B=NamedRerouteUsage(), Alpha=NamedRerouteUsage()) {constA=1}
  n21 = NamedRerouteUsage()
  NamedRerouteDeclaration(None=Multiply(
    A=MaterialFunctionCall(Corner Radius=@n18, Radius in Pixels=@n19) {materialFunction=&MF_UI_CoreUI_RoundedMask}.Shape Mask
    B=LinearInterpolate(A=@n20, Alpha=@n21) {constA=1, constB=0.18}
   )) {name=Final Opacity, nodeColor=<rgba 0 0 0 1>}
  NamedRerouteUsage()
 ]
}
