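# UI material for the sidebar party button avatar: a translucent circular portrait with a pending state.
# The portrait texture sits over a radial-gradient backing, with overlay tint and incognito or empty looks using one of two silhouette masks.
# An offset circle cuts into the portrait mask when IsBehind is set, apparently for overlapping avatars.
# The isPending parameter blends color and opacity toward three procedural dots that pulse in size and brightness on staggered sine waves.
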Material /Sidebar/Materials/M_Sidebar_PartyButtonAvatar
graph {
 properties {bAutomaticallySetUsageInEditor=false, blendMode=BLEND_Translucent, materialDomain=MD_UI, usedShadingModels=Default Lit}
 nodes [
  n1 = Constant3Vector() {constant=<rgba 0.2865 0.2865 0.2865 0>}
  n2 = Constant3Vector() {constant=<rgba 0.6997 0.6997 0.6997 0>}
  n3 = LinearInterpolate(A=@n1, B=@n2, Alpha=NamedRerouteUsage())
  n4 = NamedRerouteUsage()
  n5 = LinearInterpolate(A=@n1, B=@n2, Alpha=NamedRerouteUsage())
  n6 = NamedRerouteUsage()
  n7 = Add(
   A=NamedRerouteDeclaration(None=Multiply(A=@n3, B=@n4)) {name=DotColor, nodeColor=<rgba 0 0.8359 1 1>}
   B=NamedRerouteDeclaration(None=Multiply(A=@n5, B=@n6)) {name=DotColor_1, nodeColor=<rgba 0 0.8359 1 1>}
  )
  n8 = NamedRerouteUsage()
  n9 = NamedRerouteDeclaration(None=Multiply(A=LinearInterpolate(A=@n1, B=@n2, Alpha=@n8), B=NamedRerouteUsage())) {name=DotColor_2, nodeColor=<rgba 0 0.8359 1 1>}
  n10 = Constant3Vector() {constant=<rgba 0.25 0.25 0.25 0>}
  n11 = NamedRerouteUsage()
  LinearInterpolate(A=NamedRerouteUsage(), B=Add(A=Add(A=@n7, B=@n9), B=Multiply(A=@n10, B=@n11)), Alpha=ScalarParameter() {parameterName=isPending}) {constA=0.3, constAlpha=0}
  n12 = NamedRerouteUsage()
  n13 = LinearInterpolate(Alpha=NamedRerouteUsage()) {constA=0.3}
  n14 = NamedRerouteUsage()
  n15 = LinearInterpolate(Alpha=NamedRerouteUsage()) {constA=0.3}
  n16 = NamedRerouteUsage()
  n17 = Add(
   A=Add(A=Multiply(A=@n12, B=@n13), B=Multiply(A=@n14, B=@n15))
   B=Multiply(A=NamedRerouteUsage(), B=LinearInterpolate(Alpha=@n16) {constA=0.3})
  )
  LinearInterpolate(A=NamedRerouteUsage(), B=Add(A=Saturate(None=@n17), B=NamedRerouteUsage()), Alpha=ScalarParameter() {parameterName=isPending}) {constA=0.3, constAlpha=0}
  n18 = TextureCoordinate()
  n19 = Constant2Vector() {g=0.55, r=0.5}
  n20 = NamedRerouteUsage()
  NamedRerouteDeclaration(None=Step(Y=Distance(A=@n18, B=@n19), X=LinearInterpolate(Alpha=@n20) {constA=0.03, constB=0.06})) {name=DotMask_1, nodeColor=<rgba 0 0.2734 1 1>}
  n21 = Constant2Vector() {g=0.55, r=0.75}
  n22 = NamedRerouteUsage()
  NamedRerouteDeclaration(None=Step(Y=Distance(A=@n18, B=@n21), X=LinearInterpolate(Alpha=@n22) {constA=0.03, constB=0.06})) {name=DotMask_2, nodeColor=<rgba 0 0.2734 1 1>}
  n23 = TextureObjectParameter() {
   group=0 - Portrait
   parameterName=FrontSmallPreview
   sortPriority=0
   texture=/Game/UI/Foundation/Textures/Icons/Heroes/Athena/Soldier/T-Soldier-HID-883-Athena-Commando-M-ChOneJonesy
  }
  NamedRerouteDeclaration(None=TextureSample(UVs=NamedRerouteUsage(), Tex=@n23) {samplerSource=SSM_Clamp_WorldGroupSettings}.RGBA) {name=Front Texture Pass, nodeColor=<rgba 0 0 0 1>}
  n24 = MaterialFunctionCall() {materialFunction=&GetUserInterfaceUV}
  NamedRerouteDeclaration(None=Subtract(A=Multiply(A=@n24.Normalized UV) {constB=1.125}) {constB=0.0625}) {name=Front Normalized UV, nodeColor=<rgba 0 0 0 1>}
  NamedRerouteDeclaration(None=NamedRerouteUsage()) {name=Color Pass, nodeColor=<rgba 0 0 0 1>}
  n25 = TextureSample(
   UVs=NamedRerouteUsage()
   Tex=TextureObject() {texture=&T-Soldier-HID-883-Athena-Commando-M-ChOneJonesy}
  ) {
   samplerSource=SSM_Clamp_WorldGroupSettings
   texture=&T-Soldier-HID-883-Athena-Commando-M-ChOneJonesy
  }
  n26 = TextureSample(
   UVs=NamedRerouteUsage()
   Tex=TextureObject() {texture=/Game/UI/Foundation/Textures/Icons/Heroes/Athena/Soldier/T-Soldier-HID-884-Athena-Commando-F-ChOneRamirez}
  ) {
   samplerSource=SSM_Clamp_WorldGroupSettings
   texture=&T-Soldier-HID-883-Athena-Commando-M-ChOneJonesy
  }
  n27 = ScalarParameter() {group=3 - Empty, parameterName=RandomizeSilhouette}
  NamedRerouteDeclaration(None=LinearInterpolate(A=@n25.A, B=@n26.A, Alpha=Floor(None=@n27))) {name=Front Incognito Mask, nodeColor=<rgba 0 0 0 1>}
  n28 = NamedRerouteDeclaration(None=Floor(None=ScalarParameter() {group=0 - Portrait, parameterName=IsFrontEmpty, sliderMax=1, sortPriority=1})) {name=Front Is Empty, nodeColor=<rgba 0 0 0 1>}
  n29 = NamedRerouteUsage()
  n30 = LinearInterpolate(
   A=NamedRerouteUsage()
   B=ComponentMask(None=NamedRerouteUsage()) {b=true, g=true, r=true}
   Alpha=Multiply(A=OneMinus(None=@n28), B=ComponentMask(None=@n29) {a=true})
  )
  n31 = Constant3Vector() {constant=<rgba 0 0.0157 0.0902 1>}
  n32 = Multiply(
   A=NamedRerouteDeclaration(None=ScalarParameter() {group=0 - Portrait, parameterName=OverlayOpacity, sliderMax=1, sortPriority=1}) {name=Front Overlay Opacity, nodeColor=<rgba 0 0 0 1>}
   B=Constant() {desc=Max Value, r=0.75}
  )
  n33 = VectorParameter() {desc=Incognito Color, parameterName=IncognitoColor, defaultValue=<rgba 0.0296 0.0296 0.0296 1>}
  n34 = VectorParameter() {desc=Inactive Incognito Color, parameterName=InactiveIncognitoColor, defaultValue=<rgba 0.0296 0.0296 0.0296 1>}
  n35 = NamedRerouteUsage()
  n36 = NamedRerouteUsage()
  n37 = NamedRerouteUsage()
  NamedRerouteDeclaration(None=LinearInterpolate(
    A=LinearInterpolate(A=@n30, B=@n31, Alpha=@n32) {constB=0}
    B=LinearInterpolate(A=@n33.RGB, B=@n34.RGB, Alpha=@n35)
    Alpha=Multiply(A=@n36, B=@n37)
   ) {constB=0}) {name=Front Portrait Color Pass, nodeColor=<rgba 0 0 0 1>}
  n38 = Time()
  n39 = Multiply(A=Add(A=@n38) {constB=0.3}) {constB=0.9}
  NamedRerouteDeclaration(None=ConstantBiasScale(None=Sine(None=@n39))) {name=AnimValue, nodeColor=<rgba 1 0.0469 0 1>}
  n40 = TextureObjectParameter() {
   group=0 - Portrait
   parameterName=BackSmallPreview
   sortPriority=0
   texture=&T-Soldier-HID-883-Athena-Commando-M-ChOneJonesy
  }
  NamedRerouteDeclaration(None=TextureSample(UVs=NamedRerouteUsage(), Tex=@n40) {samplerSource=SSM_Clamp_WorldGroupSettings}.RGBA) {name=Back Texture Pass, nodeColor=<rgba 0 0 0 1>}
  n41 = Constant2Vector() {g=0.55, r=0.25}
  n42 = NamedRerouteUsage()
  NamedRerouteDeclaration(None=Step(Y=Distance(A=@n18, B=@n41), X=LinearInterpolate(Alpha=@n42) {constA=0.03, constB=0.06})) {name=DotMask, nodeColor=<rgba 0 0.2734 1 1>}
  n43 = Multiply(A=Add(A=@n38) {constB=0.2}) {constB=0.9}
  NamedRerouteDeclaration(None=ConstantBiasScale(None=Sine(None=@n43))) {name=AnimValue_1, nodeColor=<rgba 1 0.0469 0 1>}
  NamedRerouteDeclaration(None=MaterialFunctionCall(UVs=NamedRerouteUsage(), Size=Constant() {r=1.15}) {materialFunction=&MF_UI_Circle}.Fill) {name=Front Portrait Cutout Mask, nodeColor=<rgba 0 0 0 1>}
  n44 = Multiply(A=Add(A=@n38) {constB=0.1}) {constB=0.9}
  NamedRerouteDeclaration(None=ConstantBiasScale(None=Sine(None=@n44))) {name=AnimValue_2, nodeColor=<rgba 1 0.0469 0 1>}
  n45 = MaterialFunctionCall() {materialFunction=&GetUserInterfaceUV}
  n46 = AppendVector(A=ScalarParameter() {defaultValue=0.75, parameterName=CutoutOffset}, B=Constant())
  NamedRerouteDeclaration(None=MaterialFunctionCall(UVs=Subtract(A=@n45.Normalized UV, B=@n46), Size=Constant() {r=1}) {materialFunction=&MF_UI_Circle}.Fill) {name=Cutout Mask, nodeColor=<rgba 0 0 0 1>}
  n47 = NamedRerouteUsage()
  n48 = Floor(None=ScalarParameter() {defaultValue=1, parameterName=IsBehind})
  NamedRerouteDeclaration(None=LinearInterpolate(A=NamedRerouteUsage(), Alpha=LinearInterpolate(B=@n47, Alpha=@n48) {constAlpha=0}) {constB=0}) {name=Combined Portrait Mask Cut, nodeColor=<rgba 0 0 0 1>}
  NamedRerouteDeclaration(None=MaterialFunctionCall(UVs=NamedRerouteUsage(), Size=Constant() {r=1}) {materialFunction=&MF_UI_Circle}.Fill) {name=Front Portrait Mask, nodeColor=<rgba 0 0 0 1>}
  n49 = VectorParameter() {group=0 - Portrait, parameterName=FrontBackgroundRGB, defaultValue=<rgba 0.013 0.3613 0.6376 1>}
  n50 = VectorParameter() {group=3 - Empty, parameterName=EmptyBackgroundRGB, sortPriority=0, defaultValue=<rgba 0.0578 0.0578 0.0578 1>}
  n51 = NamedRerouteUsage()
  n52 = VectorParameter() {group=0 - Portrait, parameterName=FrontHighlightRGB, defaultValue=<rgba 0.4452 1 1 1>}
  n53 = VectorParameter() {group=3 - Empty, parameterName=EmptyHighlightRGB, sortPriority=1, defaultValue=<rgba 0.1022 0.1022 0.1022 1>}
  n54 = MaterialFunctionCall(UVs=NamedRerouteUsage(), Texture Scale=Constant() {r=1.5}) {materialFunction=&ScaleUVsByCenter}
  n55 = AppendVector(A=Constant(), B=Constant() {r=-0.25})
  n56 = NamedRerouteDeclaration(None=MaterialFunctionCall(UV=Add(A=@n54.UVs, B=@n55)) {materialFunction=&MF_UI_RadialGradient}.Smooth) {name=Front Radial Gradient, nodeColor=<rgba 0 0 0 1>}
  n57 = Constant() {desc=Radial Intensity, r=0.5}
  NamedRerouteDeclaration(None=LinearInterpolate(
    A=LinearInterpolate(A=@n49.RGB, B=@n50.RGB, Alpha=@n51)
    B=LinearInterpolate(A=@n52.RGB, B=@n53.RGB, Alpha=@n51)
    Alpha=Multiply(A=@n56, B=@n57) {constB=0.5}
   )) {name=Front Backing Pass, nodeColor=<rgba 0 0 0 1>}
 ]
}
