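# UI material for the sidebar's pending-invite indicator, drawing three animated dots in a horizontal row.
# Each dot is a procedural circle mask from UV distance and Step, with no textures.
# Time-offset sine waves pulse each dot's radius and appear to drive its grey shade and opacity through named reroutes, giving a staggered wave.
# The dots' colours and alphas are summed separately, with alpha saturated, for alpha-composite blending.
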Material /Sidebar/Materials/M_Sidebar_PendingInvite
graph {
 properties {blendMode=BLEND_AlphaComposite, materialDomain=MD_UI, pixelDepthOffsetMode=PDOM_Legacy, usedShadingModels=Default Lit}
 nodes [
  n1 = Constant3Vector() {constant=<rgba 0.2865 0.2865 0.2865 0>}
  n2 = Constant3Vector() {constant=<rgba 0.6997 0.6997 0.6997 0>}
  n3 = Multiply(A=LinearInterpolate(A=@n1, B=@n2, Alpha=NamedRerouteUsage()), B=NamedRerouteUsage())
  n4 = Multiply(A=LinearInterpolate(A=@n1, B=@n2, Alpha=NamedRerouteUsage()), B=NamedRerouteUsage())
  n5 = LinearInterpolate(A=@n1, B=@n2, Alpha=NamedRerouteUsage())
  n6 = NamedRerouteUsage()
  Add(
   A=Add(
    A=NamedRerouteDeclaration(None=@n3) {name=DotColor, nodeColor=<rgba 0 0.8359 1 1>}
    B=NamedRerouteDeclaration(None=@n4) {name=DotColor_1, nodeColor=<rgba 0 0.8359 1 1>}
   )
   B=NamedRerouteDeclaration(None=Multiply(A=@n5, B=@n6)) {name=DotColor_2, nodeColor=<rgba 0 0.8359 1 1>}
  )
  n7 = Time()
  n8 = Multiply(A=Add(A=@n7) {constB=0.3}) {constB=0.9}
  NamedRerouteDeclaration(None=ConstantBiasScale(None=Sine(None=@n8))) {name=AnimValue, nodeColor=<rgba 1 0.0469 0 1>}
  n9 = TextureCoordinate()
  n10 = Constant2Vector() {g=0.5, r=0.23}
  n11 = NamedRerouteUsage()
  NamedRerouteDeclaration(None=Step(Y=Distance(A=@n9, B=@n10), X=LinearInterpolate(Alpha=@n11) {constA=0.07, constB=0.1})) {name=DotMask, nodeColor=<rgba 0 0.2734 1 1>}
  n12 = Multiply(A=Add(A=@n7) {constB=0.2}) {constB=0.9}
  NamedRerouteDeclaration(None=ConstantBiasScale(None=Sine(None=@n12))) {name=AnimValue_1, nodeColor=<rgba 1 0.0469 0 1>}
  n13 = Constant2Vector() {g=0.5, r=0.5}
  n14 = NamedRerouteUsage()
  NamedRerouteDeclaration(None=Step(Y=Distance(A=@n9, B=@n13), X=LinearInterpolate(Alpha=@n14) {constA=0.07, constB=0.1})) {name=DotMask_1, nodeColor=<rgba 0 0.2734 1 1>}
  n15 = Multiply(A=Add(A=@n7) {constB=0.1}) {constB=0.9}
  NamedRerouteDeclaration(None=ConstantBiasScale(None=Sine(None=@n15))) {name=AnimValue_2, nodeColor=<rgba 1 0.0469 0 1>}
  n16 = Constant2Vector() {g=0.5, r=0.77}
  n17 = NamedRerouteUsage()
  NamedRerouteDeclaration(None=Step(Y=Distance(A=@n9, B=@n16), X=LinearInterpolate(Alpha=@n17) {constA=0.07, constB=0.1})) {name=DotMask_2, nodeColor=<rgba 0 0.2734 1 1>}
  n18 = Multiply(A=NamedRerouteUsage(), B=LinearInterpolate(Alpha=NamedRerouteUsage()) {constA=0.3})
  n19 = Multiply(A=NamedRerouteUsage(), B=LinearInterpolate(Alpha=NamedRerouteUsage()) {constA=0.3})
  n20 = NamedRerouteUsage()
  n21 = LinearInterpolate(Alpha=NamedRerouteUsage()) {constA=0.3}
  Saturate(None=Add(A=Add(A=@n18, B=@n19), B=Multiply(A=@n20, B=@n21)))
 ]
}
