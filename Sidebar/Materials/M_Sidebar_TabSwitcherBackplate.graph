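# Translucent UI material that draws the sidebar tab switcher's backplate: a rounded container with a rounded highlight for the selected tab.
# The highlight's width derives from the Number of Tabs, Tab Padding and Tab Spacing parameters, with pixel values converted to normalized UV space.
# The Control parameter slides the highlight horizontally between mirrored end offsets.
# CoreUI rounded-mask functions shape both plates, with separate corner radii, and lerps set color and faint opacity.
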Material /Sidebar/Materials/M_Sidebar_TabSwitcherBackplate
graph {
 properties {bAutomaticallySetUsageInEditor=false, blendMode=BLEND_Translucent, materialDomain=MD_UI, usedShadingModels=Default Lit}
 nodes [
  LinearInterpolate(B=Constant3Vector() {constant=<rgba 0.3864 0.3763 0.4793 1>}, Alpha=NamedRerouteUsage())
  Multiply(A=LinearInterpolate(Alpha=NamedRerouteUsage()) {constA=0.2, constB=0.18}, B=NamedRerouteUsage())
  n1 = Constant() {r=1}
  n2 = NamedRerouteUsage()
  n3 = Subtract(A=Multiply(A=Subtract(A=@n1, B=@n2)) {constB=0.5}, B=NamedRerouteUsage())
  n4 = Subtract(
   A=ComponentMask(None=NamedRerouteUsage()) {r=true}
   B=LinearInterpolate(A=Multiply(A=@n3) {constB=-1}, B=@n3, Alpha=ScalarParameter() {parameterName=Control, sliderMax=1}) {constA=-0.25, constB=0.25}
  ) {constB=0}
  n5 = ComponentMask(None=NamedRerouteUsage()) {g=true}
  n6 = Add(A=NamedRerouteUsage(), B=NamedRerouteUsage())
  n7 = Subtract(
   A=NamedRerouteUsage()
   B=MaterialFunctionCall!(Value PX=Multiply(A=@n6) {constB=0.5}, UV Pixel Size, X or Y=Constant()) {materialFunction=/FortUILibrary/MaterialFunctions/Basics/MF_UI_PixelsToNormalized}.Normalized
  ) {constA=0.4}
  n8 = OneMinus(None=MaterialFunctionCall(Value PX=NamedRerouteUsage(), X or Y=Constant() {r=1}) {materialFunction=&MF_UI_PixelsToNormalized}.Normalized)
  NamedRerouteDeclaration(None=MaterialFunctionCall(
    UV=AppendVector(A=@n4, B=@n5)
    Size=AppendVector(A=@n7, B=@n8)
    Corner Radius=NamedRerouteUsage()
    Radius in Pixels=Constant() {r=1}
    Corners Modulated by Size=Constant()
   ) {materialFunction=&MF_UI_CoreUI_RoundedMask}.Shape Mask) {name=Selected Backplate, nodeColor=<rgba 1 0.9141 0 1>}
  NamedRerouteDeclaration(None=MaterialFunctionCall(Corner Radius=NamedRerouteUsage(), Radius in Pixels=Constant() {r=1}, Corners Modulated by Size=Constant()) {materialFunction=&MF_UI_CoreUI_RoundedMask}.Shape Mask) {name=Backplate, nodeColor=<rgba 0.4062 0 1 1>}
  n9 = ScalarParameter() {defaultValue=2, parameterName=Number of Tabs}
  NamedRerouteDeclaration(None=Divide(B=Max(A=@n9))) {name=Width, nodeColor=<rgba 0.875 1 0 1>}
  NamedRerouteDeclaration(None=ScalarParameter() {defaultValue=16, parameterName=Tab Padding}) {name=Tab Padding, nodeColor=<rgba 1 0 0.2109 1>}
  NamedRerouteDeclaration(None=ScalarParameter() {defaultValue=4, parameterName=Tab Spacing}) {name=Tab Spacing, nodeColor=<rgba 0 0.5312 1 1>}
  NamedRerouteDeclaration(None=MaterialFunctionCall() {materialFunction=&GetUserInterfaceUV}.Normalized UV) {name=Normalized UV, nodeColor=<rgba 0.8516 1 0 1>}
  n10 = NamedRerouteUsage()
  n11 = NamedRerouteUsage()
  n12 = Subtract(A=NamedRerouteUsage(), B=Multiply(A=Add(A=@n10, B=@n11)) {constB=0.5})
  NamedRerouteDeclaration(None=MaterialFunctionCall(Value PX=Multiply(A=@n12) {constB=0.5}, X or Y=Constant()) {materialFunction=&MF_UI_PixelsToNormalized}.Normalized) {name=Tab Spacing Offset, nodeColor=<rgba 0.8281 0 1 1>}
  NamedRerouteDeclaration(None=ScalarParameter() {defaultValue=18, parameterName=Selected Backplate Corner Radius}) {name=Selected Backplate Corner Radius, nodeColor=<rgba 0.2656 0 1 1>}
  NamedRerouteDeclaration(None=ScalarParameter() {defaultValue=24, parameterName=Container Backplate Corner Radius}) {name=Container Backplate Corner Radius, nodeColor=<rgba 0.0547 1 0 1>}
 ]
}
