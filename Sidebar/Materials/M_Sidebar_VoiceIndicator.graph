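# Alpha-composite UI material for the sidebar voice indicator, drawing an animated green-tinted circular outline and glow, with an eased Anim Control as the voice-active state.
# An SDF circle supplies anti-aliased outline and glow masks, with an offset circle for the Display Presence area.
# Panning polar-space noise with explicit derivatives and a screen-space flow map add rift-like pulsing.
# Oscillating colors blend into the tint unless UseFlatColor is set; hover parameters modulate alpha.
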Material /Sidebar/Materials/M_Sidebar_VoiceIndicator
graph {
 properties {blendMode=BLEND_AlphaComposite, materialDomain=MD_UI, pixelDepthOffsetMode=PDOM_Legacy, usedShadingModels=Default Lit}
 nodes [
  Multiply(A=NamedRerouteUsage(), B=Multiply(A=VertexColor().A, B=NamedRerouteUsage()))
  n1 = MaterialFunctionCall(UVs=NamedRerouteUsage(), Size=NamedRerouteUsage(), OutlineThickness=NamedRerouteUsage(), OutlineAlignment=Constant() {r=-1}) {materialFunction=&MF_UI_Circle}
  NamedRerouteDeclaration(None=MaterialFunctionCall(DistanceField=NamedRerouteDeclaration(None=@n1.SDFOutline) {name=SDF Outline, nodeColor=<rgba 0 0 0 1>}, Smoothing=Constant() {r=0.03}) {materialFunction=&MF_UI_AntiAliasing}.Inner) {name=Outline Mask, nodeColor=<rgba 0 0 0 1>}
  n2 = Constant() {r=4}
  n3 = NamedRerouteUsage()
  NamedRerouteDeclaration(None=MaterialFunctionCall!(Value PX=LinearInterpolate(B=@n2, Alpha=@n3) {constA=1}, UV Pixel Size=NamedRerouteUsage(), X or Y) {materialFunction=/FortUILibrary/MaterialFunctions/Basics/MF_UI_PixelsToNormalized}.Normalized) {name=Outline Thickness, nodeColor=<rgba 0 0 0 1>}
  n4 = MaterialFunctionCall() {materialFunction=&GetUserInterfaceUV}
  NamedRerouteDeclaration(None=@n4.Pixel Size) {name=Pixel Size, nodeColor=<rgba 0 0 0 1>}
  n5 = NamedRerouteUsage()
  n6 = Sine(None=Multiply(A=Time()) {constB=0.125})
  n7 = LinearInterpolate(
   A=NamedRerouteUsage()
   B=AppendVector(A=OneMinus(None=@n5), B=@n5)
   Alpha=LinearInterpolate(Alpha=ConstantBiasScale(None=@n6)) {constA=0.95, constB=0.85}
  )
  n8 = NamedRerouteUsage()
  n9 = MaterialFunctionCall!(Vector or UVs=NamedRerouteDeclaration(None=MaterialFunctionCall!(Value=@n7, Scale=@n8, Height=@n8) {materialFunction=/FortUILibrary/MaterialFunctions/Basics/MF_UI_PingPong}.Result) {name=PrepPolarUVs, nodeColor=<rgba 0 0 0 1>}) {materialFunction=/FortUILibrary/MaterialFunctions/Basics/MF_UI_CheaperVectorToRadial}
  n10 = ComponentMask(None=@n9.Radial Coordinates) {r=true}
  n11 = ComponentMask(None=@n9.Radial Coordinates) {g=true}
  n12 = AppendVector(A=Multiply(A=@n10) {constB=3}, B=Multiply(A=@n11) {constB=2.75})
  n13 = Time()
  n14 = Constant() {r=-0.05}
  n15 = AppendVector(A=Constant() {r=0.1}, B=@n14)
  n16 = Multiply(A=@n11) {constB=1.5}
  n17 = NamedRerouteUsage()
  n18 = TextureSample!(
   UVs=Panner(Coordinate=AppendVector(A=@n10, B=@n16), Speed=@n14)
   Tex=NamedRerouteDeclaration(None=TextureObject() {desc=Rift Noise Texture, samplerType=SAMPLERTYPE_Grayscale, texture=/FortUILibrary/UIKit/Buttons/Textures/LowResBlurredNoise}) {name=Rift Noise Texture, nodeColor=<rgba 0 0 0 1>}
   "DDX(UVs)"=NamedRerouteDeclaration(None=DDX(Value=@n17)) {name=Polar DDX, nodeColor=<rgba 0 0 0 1>}
   "DDY(UVs)"=NamedRerouteDeclaration(None=DDY(Value=@n17)) {name=Polar DDY, nodeColor=<rgba 0 0 0 1>}
   Apply View MipBias
  ) {mipValueMode=TMVM_Derivative}
  n19 = Constant() {r=0.5}
  n20 = TextureSample(
   UVs=Add(A=Panner(Coordinate=@n12, Time=@n13, Speed=@n15), B=Multiply(A=@n18.R, B=@n19))
   Tex=NamedRerouteUsage()
   "DDX(UVs)"=NamedRerouteUsage()
   "DDY(UVs)"=NamedRerouteUsage()
  ) {mipValueMode=TMVM_Derivative}
  NamedRerouteDeclaration(None=SmoothStep(Value=DotProduct(A=@n20.RGB, B=@n18.RGB))) {name=Rift Pulse Mask, nodeColor=<rgba 0 0 0 1>}
  n21 = Rotator(Coordinate=NamedRerouteUsage(), Time=Multiply(A=Time()) {constB=0.333}) {speed=6.2832}
  NamedRerouteDeclaration(None=LinearInterpolate(
    A=VectorParameter() {group=Color, parameterName=Color_Outline, defaultValue=<rgba 0 1 0.3767 1>}.RGB
    B=MaterialFunctionCall!(PrimaryUV, SecondaryUV, UVMix, PixelSize) {materialFunction=/FortUILibrary/MaterialFunctions/CoreUI/MF_UI_CoreUI_OscillatingColor}.Result
    Alpha=ComponentMask(None=@n21) {r=true}
   ) {constA=1}) {name=RiftOscillatingColors, nodeColor=<rgba 0 0 0 1>}
  n22 = NamedRerouteUsage()
  n23 = LinearInterpolate(Alpha=NamedRerouteUsage()) {constA=-0.1, constB=0}
  NamedRerouteDeclaration(None=OneMinus(None=Add(A=@n22, B=@n23) {constB=0})) {name=Outline Size, nodeColor=<rgba 0 0 0 1>}
  NamedRerouteDeclaration(None=@n4.Normalized UV) {name=NormalizedUV, nodeColor=<rgba 0 0 0 1>}
  NamedRerouteDeclaration(None=ScalarParameter() {defaultValue=12, group=Padding, parameterName=Padding}) {name=Padding, nodeColor=<rgba 0 0 0 1>}
  n24 = LinearInterpolate(
   A=NamedRerouteUsage()
   B=LinearInterpolate(
    A=Constant3Vector() {constant=<rgba 0.0365 1 0.729 1>}
    B=NamedRerouteUsage()
    Alpha=Constant() {desc="RiftEnergyOscillatingColorAmount`", r=0.5}
   )
   Alpha=NamedRerouteUsage()
  )
  n25 = LinearInterpolate(
   A=LinearInterpolate(A=NamedRerouteUsage(), B=Constant3Vector() {constant=<rgba 0.0997 1 0.3403 1>}, Alpha=Constant() {desc=RiftEnergyColorTintMix, r=0.8})
   B=Constant3Vector() {desc=Outline Color, constant=<rgba 0 1 0.369 1>}
   Alpha=OneMinus(None=NamedRerouteUsage())
  )
  n26 = NamedRerouteUsage()
  NamedRerouteDeclaration(None=StaticSwitchParameter(
    True=VectorParameter() {group=Color, parameterName=Color_Outline, defaultValue=<rgba 0 1 0.3767 1>}.RGB
    False=LinearInterpolate(A=@n24, B=@n25, Alpha=@n26)
   ) {parameterName=UseFlatColor}) {name=FinalColor, nodeColor=<rgba 0 0 0 1>}
  n27 = ScalarParameter() {defaultValue=1, group=Alpha, parameterName=Glow_Alpha, sliderMax=1}
  n28 = Multiply(A=Time(), B=Constant() {desc=Glow_Alpha_Change_Speed, r=0.25})
  n29 = LinearInterpolate(A=Add(A=@n27) {constB=0.05}, B=Add(A=@n27) {constB=-0.05}, Alpha=ConstantBiasScale(None=Sine(None=@n28)))
  n30 = ScalarParameter() {defaultValue=1, group=Anim, parameterName=Anim_Hover_Fill, sliderMax=1}
  n31 = NamedRerouteUsage()
  n32 = ScalarParameter() {defaultValue=1, group=Anim, parameterName=Anim_Hover_Energy, sliderMax=1}
  n33 = NamedRerouteUsage()
  n34 = LinearInterpolate(
   A=Multiply(A=LinearInterpolate(B=@n29, Alpha=@n30) {constA=1}, B=NamedRerouteUsage())
   Alpha=Multiply(A=Multiply(A=@n31, B=@n32), B=LinearInterpolate(Alpha=@n33) {constA=0.75})
  )
  n35 = NamedRerouteUsage()
  n36 = NamedRerouteDeclaration(None=Multiply(A=ScalarParameter() {defaultValue=1, group=Alpha, parameterName=Glow_Alpha, sliderMax=1}, B=Multiply(A=@n34, B=@n35))) {name=Rift Glow Alpha, nodeColor=<rgba 0 0 0 1>}
  n37 = NamedRerouteUsage()
  NamedRerouteDeclaration(None=Multiply(A=LinearInterpolate(A=@n36, Alpha=@n37), B=ScalarParameter() {defaultValue=1, group=Alpha, parameterName=FInalAlpha, sliderMax=1})) {name=FinalAlpha, nodeColor=<rgba 0 0 0 1>}
  n38 = Multiply(A=NamedRerouteUsage()) {constB=0.5}
  n39 = SmoothStep(Min=Multiply(A=@n38) {constB=-1}, Max=@n38, Value=NamedRerouteUsage()) {constMax=0.05}
  n40 = Multiply(A=LinearInterpolate(B=OneMinus(None=@n39), Alpha=NamedRerouteUsage()))
  n41 = NamedRerouteUsage()
  NamedRerouteDeclaration(None=LinearInterpolate(A=@n40, B=LinearInterpolate(A=@n40, Alpha=@n41) {constB=0}, Alpha=ScalarParameter() {defaultValue=1, parameterName=Display Presence})) {name=Glow Mask, nodeColor=<rgba 0 0 0 1>}
  n42 = NamedRerouteUsage()
  n43 = Constant() {desc="Needs to be 64 / (64+padding) * .25", r=0.2}
  NamedRerouteDeclaration(None=Add(A=MaterialFunctionCall(UVs=@n42, Texture Scale=@n43) {materialFunction=&ScaleUVsByCenter}.UVs) {constB=-1.3}) {name=Presence UV, nodeColor=<rgba 0 0 0 1>}
  n44 = NamedRerouteDeclaration(None=LinearInterpolate(Alpha=NamedRerouteUsage()) {constA=0.5}) {name=Reduced Flow Noise, nodeColor=<rgba 0 0 0 1>}
  n45 = NamedRerouteUsage()
  n46 = NamedRerouteUsage()
  n47 = LinearInterpolate(
   A=NamedRerouteUsage()
   B=LinearInterpolate(A=LinearInterpolate(B=@n45, Alpha=@n46), Alpha=NamedRerouteUsage()) {constB=0}
   Alpha=ScalarParameter() {defaultValue=1, parameterName=Display Presence}
  )
  NamedRerouteDeclaration(None=Multiply(A=Multiply(A=@n44, B=@n47), B=ScalarParameter() {defaultValue=1, group=Anim, parameterName=Anim_Hover_Outline, sliderMax=1})) {name=Outline Alpha, nodeColor=<rgba 0 0 0 1>}
  NamedRerouteDeclaration(None=MaterialFunctionCall("Value (0-1)"=ScalarParameter() {defaultValue=1, parameterName=Anim Control}) {materialFunction=&MF_UI_EaseCurves}.EaseInCubic) {name=Is Voice Active, nodeColor=<rgba 1 0.6563 0 1>}
  NamedRerouteDeclaration(None=MaterialFunctionCall(UVs=NamedRerouteUsage(), Size=Constant() {r=1.1}) {materialFunction=&MF_UI_Circle}.Fill) {name=Presence Mask, nodeColor=<rgba 0 0 0 1>}
  NamedRerouteDeclaration(None=MaterialFunctionCall(Value PX=NamedRerouteUsage(), UV Pixel Size=NamedRerouteUsage()) {materialFunction=&MF_UI_PixelsToNormalized}.Normalized) {name=Normalized Padding, nodeColor=<rgba 0 0 0 1>}
  n48 = ScreenPosition()
  n49 = Add(A=NamedRerouteUsage())
  n50 = Time()
  n51 = Multiply(A=@n50) {constB=0.05}
  n52 = Add(A=Divide(A=@n48.ViewportUV, B=Multiply(A=@n49) {constB=2}), B=AppendVector(A=Multiply(A=@n50) {constB=-0.1}, B=@n51))
  n53 = NamedRerouteUsage()
  n54 = Divide(A=@n48.ViewportUV, B=Multiply(A=@n49) {constB=3})
  n55 = AppendVector(A=Multiply(A=@n50) {constB=0.025}, B=@n51)
  n56 = AppendVector(
   A=TextureSample!(UVs=NamedRerouteDeclaration(None=@n52) {name=PanningUVs, nodeColor=<rgba 0 0 0 1>}, Tex=@n53, Apply View MipBias).R
   B=TextureSample(UVs=Add(A=@n54, B=@n55), Tex=@n53).R
  )
  n57 = Constant() {r=0.2}
  n58 = MaterialFunctionCall!(
   UVs=NamedRerouteUsage()
   Texture Object=NamedRerouteUsage()
   Grayscale Flow Map=@n56
   Flow Direction=OneMinus(None=ComponentMask(None=@n56) {r=true})
   Flow Intensity=@n57
   Flow Speed=@n57
  ) {materialFunction=/FortUILibrary/MaterialFunctions/Effects/MF_UI_GrayscaleFlowMap_New}
  NamedRerouteDeclaration(None=SmoothStep(Value=ComponentMask(None=@n58.Distorted Texture) {r=true})) {name=Flowing Noise Mask, nodeColor=<rgba 0 0 0 1>}
 ]
}
