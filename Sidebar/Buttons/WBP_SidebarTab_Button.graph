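# Sidebar tab button widget derived from WBP_Sidebar_ButtonGeneric, showing an icon and a numeric status indicator.
# PreConstruct scales the inherited StatusIndicator according to IsUsingCompactUI, using full scale in compact mode and a smaller one otherwise.
# SetPreviewData casts the MVVM viewmodel to UIKitSelectableButtonViewModel, then applies the sidebar icon material and the viewmodel's icon.
# UpdateIndicatorCount overrides the count and collapses non-positive values, while UpdateIndicatorLimit overrides the count ceiling.
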WidgetBlueprint /Sidebar/Buttons/WBP_SidebarTab_Button
graphs [
 EventGraph {nodes [
   n1 = Event PreConstruct()
   > Parent: PreConstruct(IsDesignTime=@n1.IsDesignTime)
   > SetRenderScale!(
    execute
    self:Widget Object Reference=Get StatusIndicator() -> (StatusIndicator:WBP UIKit Status Indicator Object Reference)
    Scale:Vector 2D Structure=Select(Option 0:Vector 2D Structure=(X=0.75,Y=0.75), Option 1:Vector 2D Structure=(X=1,Y=1), Index=IsUsingCompactUI()) -> (ReturnValue:Vector 2D Structure)
   ) -> (then)
   Event Construct()
   > Parent: Construct()
  ]}
 SetPreviewData {nodes [
   n1 = SetPreviewData!() -> (then, VM:MVVM Base Viewmodel Object Reference) {is=title K2Node_FunctionEntry}
   > n2 = Cast To UIKitSelectableButtonViewModel!(execute, Object=@n1) -> (then, CastFailed, AsUI Kit Selectable Button View Model:UI Kit Selectable Button View Model Object Reference) {is=title K2Node_DynamicCast}
   > SetIconMaterial!(
    execute
    self:WBP UIKit Button Generic Base Object Reference
    Material Instance:Material Instance Object Reference=/Sidebar/Materials/Icons/MI_Sidebar_Icon_Base
   ) -> (then)
   > SetIcon!(
    execute
    self:WBP UIKit Button Generic Base Object Reference
    Icon:Object Reference=Get Icon!(self:UI Kit Selectable Button View Model Object Reference=@n2) -> (Icon:Texture 2D Object Reference)
   ) -> (then)
  ]}
 UpdateIndicatorCount {nodes [
   n1 = UpdateIndicatorCount!() -> (then, Count:int) {is=title K2Node_FunctionEntry}
   n2 = Get StatusIndicator()
   > OverrideCount!(execute, self:WBP UIKit Status Indicator Base Object Reference=@n2, Count:int=@n1) -> (then)
   > SetVisibility(
    self=@n2
    InVisibility=Select(Option 0:ESlateVisibility Enum=Collapsed, Option 1:ESlateVisibility Enum=HitTestInvisible, Index="integer > integer"(A=@n1)) -> (ReturnValue:ESlateVisibility Enum)
   )
  ]}
 UpdateIndicatorLimit {nodes [
   n1 = UpdateIndicatorLimit!() -> (then, Limit:int) {is=title K2Node_FunctionEntry}
   > OverrideCountCeiling!(execute, self:WBP UIKit Status Indicator Base Object Reference=Get StatusIndicator(), Ceiling:int=@n1) -> (then)
  ]}
]
variables []
widgets {
 info {
  inheritedWidgetCount=0
  parentClass=&/WBP_Sidebar_ButtonGeneric.WBP_Sidebar_ButtonGeneric_C
  rootWidgetClass=null
  widgetCount=0
 }
 widgets []
}
