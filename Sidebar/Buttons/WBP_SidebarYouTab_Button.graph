WidgetBlueprint /Sidebar/Buttons/WBP_SidebarYouTab_Button
graphs [
 EventGraph {nodes [
   n1 = Event PreConstruct()
   > Parent: PreConstruct(IsDesignTime=@n1.IsDesignTime)
   > SetGridSlotAlignment!(
    execute
    Widget:Widget Object Reference=Get Slot_Main() -> (Slot_Main:Named Slot Object Reference)
    Horizontal:EHorizontalAlignment Enum=HAlign_Center
    Vertical:EVerticalAlignment Enum=VAlign_Center
   ) -> (then)
   Event Construct()
   > Parent: Construct()
   n2 = GetFortPlayerControllerFromActor!(Actor:Actor Object Reference=GetOwningPlayer()) -> (ReturnValue:Fort Player Controller Object Reference)
   > InitializeWithSocialUsers!(
    execute
    self:WBP Sidebar Player Avatar Object Reference=Get LocalPlayerAvatar() -> (LocalPlayerAvatar:WBP Sidebar Player Avatar Object Reference)
    SocialUsers:Array of Fort Social User Object References=Make Array("[0]":Fort Social User Object Reference=GetSocialUserFromController!(FortPC:Fort Player Controller Object Reference=@n2) -> (ReturnValue:Fort Social User Object Reference)) -> (Array:Array of Fort Social User Object References)
   ) -> (then)
   Event On Hovered()
   > Parent: On Hovered() {is=title K2Node_CallParentFunction}
   n3 = Get CommonInputType!(self:Fort Input System VM Object Reference=Get FortInputSystemVM() -> (FortInputSystemVM:Fort Input System VM Object Reference)) -> (CommonInputType:ECommonInputType Enum)
   > Branch(Condition=AND Boolean(A="Equal (Enum)"(A=@n3, B=MouseAndKeyboard), B=GetSelected!(self:Common Button Base Object Reference) -> (ReturnValue:bool)))
   > UpdateVisibility!(execute, self, ShouldShowAvatar:bool=false) -> (then)
   Event On Unhovered()
   > Parent: On Unhovered() {is=title K2Node_CallParentFunction}
   > n4 = Branch(Condition=IsVisible(self=Get LocalPlayerAvatar()))
   UpdateVisibility(execute=@n4.else, ShouldShowAvatar=true)
   Event On Selected()
   > Parent: On Selected() {is=title K2Node_CallParentFunction}
   n5 = "Equal (Enum)"(A=Get CommonInputType(self=Get FortInputSystemVM()), B=Touch)
   > Branch(Condition=@n5)
   > UpdateVisibility(ShouldShowAvatar=false)
   Event On Deselected()
   > Parent: On Deselected() {is=title K2Node_CallParentFunction}
   > Branch(Condition=@n5)
   > UpdateVisibility(ShouldShowAvatar=true)
   n6 = Event On Input Method Changed!() -> (OutputDelegate, then, CurrentInputType:ECommonInputType Enum) {is=title K2Node_Event}
   > Parent: On Input Method Changed!(execute, CurrentInputType:ECommonInputType Enum=@n6.CurrentInputType) -> (then) {is=title K2Node_CallParentFunction}
   n7 = Get FortInputSystemVM()
   > Branch(Condition="Equal (Enum)"(A=Get CommonInputType(self=@n7), B=Gamepad))
   > UpdateVisibility(ShouldShowAvatar=true)
  ]}
 SetPreviewData {nodes [
   n1 = SetPreviewData!() -> (then, VM:MVVM Base Viewmodel Object Reference) {is=title K2Node_FunctionEntry}
   > n2 = Cast To UIKitSelectableButtonViewModel!(execute, Object=@n1) -> (then, CastFailed, AsUI Kit Selectable Button View Model:UI Kit Selectable Button View Model Object Reference) {is=title K2Node_DynamicCast}
   > SetIconMaterial!(
    execute
    self:WBP UIKit Button Generic Base Object Reference
    Material Instance:Material Instance Object Reference=/Sidebar/Materials/Icons/MI_Sidebar_Icon_Base
   ) -> (then)
   > SetIcon!(
    execute
    self:WBP UIKit Button Generic Base Object Reference
    Icon:Object Reference=Get Icon!(self:UI Kit Selectable Button View Model Object Reference=@n2) -> (Icon:Texture 2D Object Reference)
   ) -> (then)
  ]}
 UpdateVisibility {nodes [
   n1 = UpdateVisibility!() -> (then, ShouldShowAvatar:bool=~) {is=title K2Node_FunctionEntry}
   > SetVisibility(
    self=[Get Close() -> (Close:Image Object Reference), Get Cross() -> (Cross:Image Object Reference)]
    InVisibility=Select(Option 0=SelfHitTestInvisible, Option 1=Collapsed, Index=@n1)
   )
   > SetVisibility(self=Get LocalPlayerAvatar(), InVisibility=Select(Option 0=Collapsed, Option 1=SelfHitTestInvisible, Index=@n1))
  ]}
]
variables [
 PresenceColor_Away {type=LinearColor, default=<rgba 0.9647 0.5457 0.1221 1>}
 PresenceColor_Offline {type=LinearColor, default=<rgba 0.4287 0.6654 0.7991 0.302>}
 PresenceColor_Online {type=LinearColor, default=<rgba 0.2159 1 0.1329 1>}
 PresenceSelect_Away {type=LinearColor, default=<rgba 0 1 0 0>}
 PresenceSelect_Blocked {type=LinearColor, default=<rgba 0 0 0 1>}
 PresenceSelect_Offline {type=LinearColor, default=<rgba 0 0 1 0>}
 PresenceSelect_Online {type=LinearColor, default=<rgba 1 0 0 0>}
]
widgets {
 info {
  inheritedWidgetCount=0
  namedSlotCount=1
  parentClass=&/WBP_Sidebar_ButtonGeneric.WBP_Sidebar_ButtonGeneric_C
  rootWidgetClass=null
  widgetCount=4
 }
 widgets [
  Overlay_0 {widgetClassPath=&Overlay, widget=Overlay_0}
  LocalPlayerAvatar {
   widgetClassPath=/Sidebar/Panels/WBP_Sidebar_PlayerAvatar
   bIsVariable=true
   parent=Overlay_0
   widget=LocalPlayerAvatar
   properties {avatarSize=<56 56>}
   slot {class=OverlaySlot, properties {horizontalAlignment=HAlign_Fill, verticalAlignment=VAlign_Fill}}
  }
  Close {
   widgetClassPath=&Image
   bIsVariable=true
   parent=Overlay_0
   widget=Close
   properties {
    visibility=Collapsed
    brush {
     bIsDynamicallyLoaded=false
     drawAs=Image
     imageType=FullColor
     mirroring=NoMirror
     resourceName=None
     resourceObject=/FortUILibrary/Foundation/Art_ToolKit/Tools/M_UI_Circle_White
     tiling=NoTile
     imageSize=<48 48>
     margin=<ltrb 0 0 0 0>
     outlineSettings {
      bUseBrushTransparency=false
      roundingType=HalfHeightRadius
      width=0
      color {colorUseRule=UseColor_Specified, specifiedColor=<rgba 0 0 0 0>}
      cornerRadii=<0 0 0 0>
     }
     tintColor {colorUseRule=UseColor_Specified, specifiedColor=<rgba 0.0212 0.0212 0.0212 1>}
     uVRegion {bIsValid=false, max=<0 0>, min=<0 0>}
    }
   }
   slot {class=OverlaySlot, properties {horizontalAlignment=HAlign_Fill, verticalAlignment=VAlign_Fill}}
  }
  Cross {
   widgetClassPath=&Image
   bIsVariable=true
   parent=Overlay_0
   widget=Cross
   properties {
    visibility=Collapsed
    brush {
     bIsDynamicallyLoaded=false
     drawAs=Image
     imageType=FullColor
     mirroring=NoMirror
     resourceName=None
     resourceObject=/FortniteCoreUI/Frontend/StoreMain/Materials/Textures/T_UI_Icon_XClose
     tiling=NoTile
     imageSize=<16 16>
     margin=<ltrb 0 0 0 0>
     outlineSettings {
      bUseBrushTransparency=false
      roundingType=HalfHeightRadius
      width=0
      color {colorUseRule=UseColor_Specified, specifiedColor=<rgba 0 0 0 0>}
      cornerRadii=<0 0 0 0>
     }
     tintColor {colorUseRule=UseColor_Specified, specifiedColor=<rgba 1 1 1 1>}
     uVRegion {bIsValid=false, max=<0 0>, min=<0 0>}
    }
   }
   slot {class=OverlaySlot, properties {horizontalAlignment=HAlign_Center, verticalAlignment=VAlign_Center}}
  }
 ]
}
