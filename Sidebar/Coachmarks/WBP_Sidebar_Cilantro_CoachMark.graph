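# Sidebar onboarding coach mark that tells players a parent or guardian has enabled time limits or reports.
# Wraps a modal UIKit coach mark, activated on show and deactivated with this widget.
# "Take me there" calls HandleDismissClicked; it and OnEndTooltip immediately report the close animation finished, then deactivate both widgets.
# SetCoachMarkPlacement picks the width for inside or outside placement, with a design-time preview in PreConstruct.
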WidgetBlueprint /Sidebar/Coachmarks/WBP_Sidebar_Cilantro_CoachMark
graphs [
 EventGraph {nodes [
   "Button Other Action Clicked (WBP_UIKit_CoachMark_Cilantro)"() {is=title K2Node_ComponentBoundEvent}
   > n1 = HandleDismissClicked(self:Fort Sidebar Onboard Tooltip Widget Object Reference)
   NotifyCloseAnimationFinished(execute=[@n1.then, @n3.then], self:Fort Sidebar Onboard Tooltip Widget Object Reference)
   > DeactivateWidget(self=Get WBP_UIKit_CoachMark_Cilantro() -> (WBP_UIKit_CoachMark_Cilantro:WBP UIKit Coach Mark Object Reference))
   > DeactivateWidget()
   Event On Deactivated()
   > DeactivateWidget(self=Get WBP_UIKit_CoachMark_Cilantro())
   n2 = Event PreConstruct()
   > Branch(Condition=@n2.IsDesignTime)
   > SetCoachMarkPlacement!(
    execute
    self
    CheckMarkPlacement:EFortSidebarCoachMarkPlacement Enum=Get PreviewCoachMarkPlacement() -> (PreviewCoachMarkPlacement:EFortSidebarCoachMarkPlacement Enum=Outside)
   ) -> (then)
   Event OnShowTooltip() {is=title K2Node_Event}
   > ActivateWidget(self=Get WBP_UIKit_CoachMark_Cilantro())
   n3 = Event OnEndTooltip() {is=title K2Node_Event}
  ]}
 SetCoachMarkPlacement {nodes [
   n1 = SetCoachMarkPlacement!() -> (then, CheckMarkPlacement:EFortSidebarCoachMarkPlacement Enum=Outside) {is=title K2Node_FunctionEntry}
   > Set Widget Width!(
    execute
    Widget Width:float=Select!(Outside:float=580, Inside:float=480, Index:EFortSidebarCoachMarkPlacement Enum=@n1) -> (ReturnValue:float)
    self:WBP UIKit Coach Mark Base Object Reference=Get WBP_UIKit_CoachMark_Cilantro()
   ) -> (then, Output_Get:float)
  ]}
]
variables [PreviewCoachMarkPlacement {type=EFortSidebarCoachMarkPlacement, default=Outside}]
widgets {
 info {
  inheritedWidgetCount=2
  parentClass=/Script/FortniteUI.FortSidebarOnboardTooltipWidget
  rootWidgetClass=/FortUILibrary/UIKit/CoachMark/WBP_UIKit_CoachMark.WBP_UIKit_CoachMark_C
  widgetCount=3
 }
 widgets [
  WBP_UIKit_CoachMark_Cilantro {
   widgetClassPath=/FortUILibrary/UIKit/CoachMark/WBP_UIKit_CoachMark
   bIsVariable=true
   widget=WBP_UIKit_CoachMark_Cilantro
   properties {
    arrowPosition=Top_Right
    bAutoRestoreFocus=true
    bIsModal=true
    navigation=&self:WidgetTree.WBP_UIKit_CoachMark_Cilantro.WidgetNavigation_0
    shouldShowDismissButton=false
    textDescription="Your parent or guardian has enabled time limits and/or time reports for your account in Fortnite and Unreal Editor for Fortnite (UEFN)."
    textHeader="Time Limits && Reports"
    textOtherActionButton=Take me there
    widget Width=580
   }
  }
  Dismiss_Button {widgetClassPath=&CommonButtonLegacy, bInherited=true, widget=null}
  OtherAction_Button {widgetClassPath=&CommonButtonLegacy, bInherited=true, widget=null}
 ]
}
