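# State Tree driving the game's sidebar UI, switching between social, chat, friend, profile, settings, exit and WebSO tabs via events.
# Root runs ST_SidebarOverlay in parallel; its event transitions jump to tabs under SidebarOpen, which can return to SidebarClosed.
# SidebarOpen registers cache entries for panel widgets, while tabs either slot widgets into UI.Sidebar.SubPanel or use linked assets.
# AddFriend and FriendRequests nest event-driven local and other player substates that return to their top panel.
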StateTree /Sidebar/StateTree/ST_Sidebar
states [
 Root {
  id=#1
  type=STATE
  children [#2, #3]
  tasks [{
    id=#4
    description=Run Parallel ST_SidebarOverlay
    properties {
     iD=#5
     instance {
      _structType=/Script/StateTreeModule.StateTreeRunParallelStateTreeTaskInstanceData
      stateTree {stateTree=&/ST_SidebarOverlay, parameters {}, propertyOverrides []}
     }
     node {
      name=None
      _structType=/Script/StateTreeModule.StateTreeRunParallelStateTreeTask
      bCanEditConsideredForCompletion=true
      bConsideredForCompletion=true
      bHasTaskCompletionDelegateDispatcher=false
      bShouldCopyParametersOnExitState=false
      bShouldCopyParametersOnTick=false
      bTaskEnabled=true
      eventHandlingPriority=Normal
      transitionHandlingPriority=Normal
      bindingsBatch {value=65535}
      executionRuntimeTemplateIndex {value=65535}
      instanceDataHandle {index=65535, source=None, stateHandle {index=65535}}
      instanceTemplateIndex {value=65535}
      outputBindingsBatch {value=65535}
      stateTreeOverrideTag {tagName=None}
     }
    }
   }]
  transitions [
   {id=#6, trigger=ON_EVENT, state {id=#7, name=Social}}
   {id=#8, trigger=ON_EVENT, state {id=#9, name=Chat}}
   {id=#10, trigger=ON_EVENT, state {id=#11, name=AddFriend}}
   {id=#12, trigger=ON_EVENT, state {id=#13, name=FriendRequests}}
   {id=#14, trigger=ON_EVENT, state {id=#15, name=Settings}}
   {id=#16, trigger=ON_EVENT, state {id=#17, name=SettingsCombined}}
   {id=#18, trigger=ON_EVENT, state {id=#19, name=WebSO}}
   {id=#20, trigger=ON_EVENT, state {id=#21, name=LocalPlayerTab}}
   {id=#22, trigger=ON_EVENT, state {id=#23, name=Exit}}
  ]
 }
 SidebarClosed {id=#2, type=STATE, tasks [], transitions []}
 SidebarOpen {
  id=#3
  type=STATE
  children [
   #24
   #7
   #21
   #9
   #11
   #13
   #15
   #17
   #23
   #19
  ]
  tasks [{
    id=#25
    description=UIState Tree Widget Caching Task
    properties {
     iD=#26
     instance {
      _structType=/Script/UIStateTree.UIStateTreeWidgetCachingTaskInstanceData
      cacheEntries [
       {cacheEntryName=SocialPanel, initialInstanceCount=0, memoryTag=None, widgetClass=/Sidebar/Panels/WBP_Social_Panel.WBP_Social_Panel_C}
       {
        cacheEntryName=AddFriendsPanel
        initialInstanceCount=0
        memoryTag=None
        widgetClass=/Sidebar/Panels/WBP_AddFriends_Panel.WBP_AddFriends_Panel_C
       }
       {cacheEntryName=SettingsPanel, initialInstanceCount=0, memoryTag=None, widgetClass=/Sidebar/Panels/WBP_Settings_Panel.WBP_Settings_Panel_C}
       {cacheEntryName=ExitPanel, initialInstanceCount=0, memoryTag=None, widgetClass=/Sidebar/Panels/WBP_Exit_Panel.WBP_Exit_Panel_C}
       {
        cacheEntryName=ChatPanel
        initialInstanceCount=0
        memoryTag=None
        widgetClass=/Sidebar/Panels/Chat/WBP_Sidebar_ChatChannelsPanel.WBP_Sidebar_ChatChannelsPanel_C
       }
       {
        cacheEntryName=OtherPlayerPanel
        initialInstanceCount=0
        memoryTag=None
        widgetClass=/Sidebar/Panels/Social/Profile/WBP_Profile_OtherPlayer.WBP_Profile_OtherPlayer_C
       }
       {
        cacheEntryName=OtherPlayerPanelNickname
        initialInstanceCount=0
        memoryTag=None
        widgetClass=/Sidebar/Panels/Social/Profile/WBP_Profile_OtherPlayer_Nickname.WBP_Profile_OtherPlayer_Nickname_C
       }
       {
        cacheEntryName=LocalPlayerPanel
        initialInstanceCount=0
        memoryTag=None
        widgetClass=/Sidebar/Panels/Social/Profile/WBP_Profile_LocalPlayer.WBP_Profile_LocalPlayer_C
       }
       {
        cacheEntryName=OverlayPanel
        initialInstanceCount=0
        memoryTag=None
        widgetClass=/Sidebar/Panels/Overlay/WBP_Sidebar_OverlayPanelLayout.WBP_Sidebar_OverlayPanelLayout_C
       }
       {
        cacheEntryName=OverlayConfirmationPanel
        initialInstanceCount=0
        memoryTag=None
        widgetClass=/Sidebar/Panels/Overlay/WBP_Confirmation_OverlayPanel.WBP_Confirmation_OverlayPanel_C
       }
       {
        cacheEntryName=MessageListPanel
        initialInstanceCount=0
        memoryTag=None
        widgetClass=/Sidebar/Panels/Chat/WBP_Sidebar_TextChatPanel.WBP_Sidebar_TextChatPanel_C
       }
       {
        cacheEntryName=InvitePanel
        initialInstanceCount=0
        memoryTag=None
        widgetClass=/Sidebar/Panels/Chat/WBP_Sidebar_ChatInvitePanel.WBP_Sidebar_ChatInvitePanel_C
       }
       {
        cacheEntryName=ManageGroupPanel
        initialInstanceCount=0
        memoryTag=None
        widgetClass=/Sidebar/Panels/Chat/WBP_Sidebar_ManageGroupPanel.WBP_Sidebar_ManageGroupPanel_C
       }
       {
        cacheEntryName=ChatAcceptPanel
        initialInstanceCount=0
        memoryTag=None
        widgetClass=/Sidebar/Panels/Chat/WBP_Sidebar_ChatAcceptPanel.WBP_Sidebar_ChatAcceptPanel_C
       }
       {
        cacheEntryName=FriendRequestsPanel
        initialInstanceCount=0
        memoryTag=None
        widgetClass=/Sidebar/Panels/WBP_FriendRequests_Panel.WBP_FriendRequests_Panel_C
       }
       {
        cacheEntryName=SettingsCombinedPanel
        initialInstanceCount=0
        memoryTag=None
        widgetClass=/Sidebar/Panels/WBP_Exit_Settings_Combined_Panel.WBP_Exit_Settings_Combined_Panel_C
       }
      ]
     }
     node {
      name=None
      _structType=/Script/UIStateTree.UIStateTreeWidgetCachingTask
      bCanEditConsideredForCompletion=true
      bConsideredForCompletion=false
      bHasTaskCompletionDelegateDispatcher=false
      bTaskEnabled=true
      transitionHandlingPriority=Normal
      bindingsBatch {value=65535}
      executionRuntimeTemplateIndex {value=65535}
      instanceDataHandle {index=65535, source=None, stateHandle {index=65535}}
      instanceTemplateIndex {value=65535}
      outputBindingsBatch {value=65535}
     }
    }
   }]
  transitions [{id=#27, trigger=ON_EVENT, state {id=#2, name=SidebarClosed}}]
 }
 None {id=#24, type=STATE, tasks [], transitions []}
 Social {id=#7, type=LINKED_ASSET, transitions []}
 LocalPlayerTab {id=#21, type=LINKED_ASSET, transitions []}
 Chat {id=#9, type=LINKED_ASSET, transitions []}
 AddFriend {
  id=#11
  type=STATE
  children [#28, #29, #30]
  tasks []
  transitions []
 }
 FriendRequests {
  id=#13
  type=STATE
  children [#31, #32, #33]
  tasks []
  transitions []
 }
 Settings {id=#15, type=LINKED_ASSET, transitions []}
 SettingsCombined {id=#17, type=LINKED_ASSET, transitions []}
 Exit {
  id=#23
  type=STATE
  tasks [{
    id=#34
    description=UIState Tree Widget Slotting Task
    properties {
     iD=#35
     instance {
      _structType=/Script/UIStateTree.UIStateTreeWidgetSlottingTaskInstanceData
      bIsTaskActive=false
      widgets [{
        bAutoActivateWidget=true
        bCacheInstance=true
        bFill=true
        bIsPlayerSpecific=true
        bIsWidgetEnabled=true
        bKeepConstructed=true
        cacheEntryName=ExitPanel
        memoryTag=None
        widget=&WBP_Exit_Panel_C
        hostTag {tagName=UI.Sidebar.SubPanel}
        padding=<ltrb 0 0 0 0>
       }]
     }
     node {
      name=None
      _structType=/Script/UIStateTree.UIStateTreeWidgetSlottingTask
      bCanEditConsideredForCompletion=true
      bConsideredForCompletion=false
      bHasTaskCompletionDelegateDispatcher=false
      bTaskEnabled=true
      transitionHandlingPriority=Normal
      bindingsBatch {value=65535}
      executionRuntimeTemplateIndex {value=65535}
      instanceDataHandle {index=65535, source=None, stateHandle {index=65535}}
      instanceTemplateIndex {value=65535}
      outputBindingsBatch {value=65535}
     }
    }
   }]
  transitions []
 }
 WebSO {
  id=#19
  type=STATE
  tasks [{
    id=#36
    description=UIState Tree Widget Slotting Task
    properties {
     iD=#37
     instance {
      _structType=&UIStateTreeWidgetSlottingTaskInstanceData
      bIsTaskActive=false
      widgets [{
        bAutoActivateWidget=true
        bCacheInstance=false
        bFill=true
        bIsPlayerSpecific=true
        bIsWidgetEnabled=true
        bKeepConstructed=true
        cacheEntryName=ExitPanel
        memoryTag=None
        widget=/Sidebar/Panels/WBP_WebSO_Panel.WBP_WebSO_Panel_C
        hostTag {tagName=UI.Sidebar.SubPanel}
        padding=<ltrb 0 0 0 0>
       }]
     }
     node {
      name=None
      _structType=&UIStateTreeWidgetSlottingTask
      bCanEditConsideredForCompletion=true
      bConsideredForCompletion=false
      bHasTaskCompletionDelegateDispatcher=false
      bTaskEnabled=true
      transitionHandlingPriority=Normal
      bindingsBatch {value=65535}
      executionRuntimeTemplateIndex {value=65535}
      instanceDataHandle {index=65535, source=None, stateHandle {index=65535}}
      instanceTemplateIndex {value=65535}
      outputBindingsBatch {value=65535}
     }
    }
   }]
  transitions []
 }
 AddFriend_Top {
  id=#28
  type=STATE
  tasks [{
    id=#38
    description=UIState Tree Widget Slotting Task
    properties {
     iD=#39
     instance {
      _structType=&UIStateTreeWidgetSlottingTaskInstanceData
      bIsTaskActive=false
      widgets [{
        bAutoActivateWidget=true
        bCacheInstance=true
        bFill=true
        bIsPlayerSpecific=true
        bIsWidgetEnabled=true
        bKeepConstructed=true
        cacheEntryName=AddFriendsPanel
        memoryTag=None
        widget=&WBP_AddFriends_Panel_C
        hostTag {tagName=UI.Sidebar.SubPanel}
        padding=<ltrb 0 0 0 0>
       }]
     }
     node {
      name=None
      _structType=&UIStateTreeWidgetSlottingTask
      bCanEditConsideredForCompletion=true
      bConsideredForCompletion=false
      bHasTaskCompletionDelegateDispatcher=false
      bTaskEnabled=true
      transitionHandlingPriority=Normal
      bindingsBatch {value=65535}
      executionRuntimeTemplateIndex {value=65535}
      instanceDataHandle {index=65535, source=None, stateHandle {index=65535}}
      instanceTemplateIndex {value=65535}
      outputBindingsBatch {value=65535}
     }
    }
   }]
  transitions [
   {id=#40, trigger=ON_EVENT, state {id=#30, name=OtherPlayer}}
   {id=#41, trigger=ON_EVENT, state {id=#29, name=LocalPlayer}}
  ]
 }
 LocalPlayer {
  id=#29
  type=LINKED_ASSET
  transitions [{
    id=#42
    priority=null
    trigger=ON_EVENT
    conditions []
    state {id=#28, name=AddFriend_Top, linkType=GOTO_STATE}
   }]
 }
 OtherPlayer {
  id=#30
  type=LINKED_ASSET
  transitions [{
    id=#43
    priority=null
    trigger=ON_EVENT
    conditions []
    state {id=#28, name=AddFriend_Top, linkType=GOTO_STATE}
   }]
 }
 FriendRequests_Top {
  id=#31
  type=STATE
  tasks [{
    id=#44
    description=UIState Tree Widget Slotting Task
    properties {
     iD=#45
     instance {
      _structType=&UIStateTreeWidgetSlottingTaskInstanceData
      bIsTaskActive=false
      widgets [{
        bAutoActivateWidget=true
        bCacheInstance=true
        bFill=true
        bIsPlayerSpecific=true
        bIsWidgetEnabled=true
        bKeepConstructed=true
        cacheEntryName=FriendRequestsPanel
        memoryTag=None
        widget=&WBP_FriendRequests_Panel_C
        hostTag {tagName=UI.Sidebar.SubPanel}
        padding=<ltrb 0 0 0 0>
       }]
     }
     node {
      name=None
      _structType=&UIStateTreeWidgetSlottingTask
      bCanEditConsideredForCompletion=true
      bConsideredForCompletion=false
      bHasTaskCompletionDelegateDispatcher=false
      bTaskEnabled=true
      transitionHandlingPriority=Normal
      bindingsBatch {value=65535}
      executionRuntimeTemplateIndex {value=65535}
      instanceDataHandle {index=65535, source=None, stateHandle {index=65535}}
      instanceTemplateIndex {value=65535}
      outputBindingsBatch {value=65535}
     }
    }
   }]
  transitions [
   {id=#46, trigger=ON_EVENT, state {id=#33, name=OtherPlayer}}
   {id=#47, trigger=ON_EVENT, state {id=#32, name=LocalPlayer}}
  ]
 }
 LocalPlayer {
  id=#32
  type=LINKED_ASSET
  transitions [{
    id=#48
    priority=null
    trigger=ON_EVENT
    conditions []
    state {id=#31, name=FriendRequests_Top, linkType=GOTO_STATE}
   }]
 }
 OtherPlayer {
  id=#33
  type=LINKED_ASSET
  transitions [{
    id=#49
    priority=null
    trigger=ON_EVENT
    conditions []
    state {id=#31, name=FriendRequests_Top, linkType=GOTO_STATE}
   }]
 }
]
