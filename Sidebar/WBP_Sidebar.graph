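# WBP_Sidebar is the activatable sidebar overlay hosting feature tabs such as Social, Chat, Settings and panels.
# It builds tabs from FortSidebarVM features filtered by input type, sending selections through navigation-point objects and a UI state tree registered on init.
# It also handles deep links, an onboarding coach mark and push-to-talk and mute sounds.
# Layout adapts to compact UI, orientation, safe zones and, behind config flags, the virtual keyboard.
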WidgetBlueprint /Sidebar/WBP_Sidebar
graphs [
 BP_GetDesiredFocusTarget {nodes [
   Get Desired Focus Target()
   > Return Node!(
    execute
    ReturnValue:Widget Object Reference=GetChildAt!(
     self:Panel Widget Object Reference=Get PanelsSlottingHost() -> (PanelsSlottingHost:Widget Slotting Host Overlay Object Reference)
     Index:int=0
    ) -> (ReturnValue:Widget Object Reference)
   )
  ]}
 EvaluateSafezones {nodes [
   EvaluateSafezones() {is=title K2Node_FunctionEntry}
   > n1 = GetSafeZonePadding!(execute) -> (
    then
    SafePadding_X:double
    SafePadding_Y:double
    SafePadding_Z:double
    SafePadding_W:double
    SafePaddingScale:Vector 2D Structure
    SpillOverPadding:Vector 4 Structure
   )
   n2 = GetViewportScale!() -> (ReturnValue:float)
   n3 = SafeDivide(A=@n1, B=@n2)
   n4 = Get!(self:Grid Slot Object Reference=SlotAsGridSlot(Widget=Get SafeZone() -> (SafeZone:Safe Zone Object Reference))) -> (Padding_Left:float=~, Padding_Top:float=~, Padding_Right:float=~, Padding_Bottom:float=~)
   > SetSidesToPad!(
    execute
    self:Safe Zone Object Reference=Get SafeZone()
    InPadLeft:bool=Select!(
     Landscape:bool=false
     Portrait:bool="float < float"(A=@n4, B=@n3)
     Index:EDisplayOrientationType Enum=Get CachedDisplayOrientation() -> (CachedDisplayOrientation:EDisplayOrientationType Enum=Landscape)
    ) -> (ReturnValue:bool)
    InPadRight:bool="float < float"(A=@n4.Padding_Right, B=SafeDivide(A=@n1.SafePadding_Z, B=@n2))
    InPadTop:bool="float < float"(A=@n4.Padding_Top, B=SafeDivide(A=@n1.SafePadding_Y, B=@n2))
    InPadBottom:bool="float < float"(A=@n4.Padding_Bottom, B=SafeDivide(A=@n1.SafePadding_W, B=@n2))
   ) -> (then)
   n5 = Get!() -> (SafeZonePadding_Left:float=~, SafeZonePadding_Top:float=~, SafeZonePadding_Right:float=~, SafeZonePadding_Bottom:float=~)
   n6 = Get SafeZone()
   n7 = Select(Landscape=true, Portrait=false, Index=Get CachedDisplayOrientation() -> (CachedDisplayOrientation=Landscape))
   n8 = Get PadLeft!(self:Safe Zone Object Reference=@n6) -> (PadLeft:bool)
   > SetPadding!(
    execute
    self:Grid Slot Object Reference=SlotAsGridSlot(Widget=@n6)
    InPadding_Left:float=SelectFloat(A=0, B=@n5, bPickA=OR Boolean(A=@n7, B=@n8))
    InPadding_Top:float=SelectFloat(A=0, B=@n5.SafeZonePadding_Top, bPickA=Get PadTop!(self:Safe Zone Object Reference=@n6) -> (PadTop:bool))
    InPadding_Right:float=SelectFloat(A=0, B=@n5.SafeZonePadding_Right, bPickA=Get PadRight!(self:Safe Zone Object Reference=@n6) -> (PadRight:bool))
    InPadding_Bottom:float=SelectFloat(A=0, B=@n5.SafeZonePadding_Bottom, bPickA=Get PadBottom!(self:Safe Zone Object Reference=@n6) -> (PadBottom:bool))
   ) -> (then)
  ]}
 EventGraph {nodes [
   n1 = "On Tab Clicked (MainTabGroup)"!() -> (OutputDelegate, then, TabId:Name=None) {is=title K2Node_ComponentBoundEvent}
   > n2 = Branch(Condition=Get isNewFeatureSelected() -> (isNewFeatureSelected:bool))
   > Set isNewFeatureSelected(isNewFeatureSelected:bool=false) -> (Output_Get:bool)
   n3 = Get Current Input Type() -> (Current Input Type:ECommonInputType Enum)
   n4 = "Equal (Enum)"(A=@n3, B=MouseAndKeyboard)
   n5 = "Equal (Enum)"(A=@n3, B=Touch)
   Branch(
    execute=@n2.else
    Condition=AND Boolean!(A="Equal (Name)"(A=@n1.TabId, B=Social), B=OR Boolean(A=@n4, B=@n5), C:bool=isUseEpicParties!() -> (UseEpicParties:bool)) -> (ReturnValue)
   )
   > RequestCloseSidebar(self:Fort Sidebar VM Object Reference=Get FortSidebarVM() -> (FortSidebarVM:Fort Sidebar VM Object Reference))
   n6 = "On Tab Selected (MainTabGroup)"!() -> (OutputDelegate, then, TabId:Name=None) {is=title K2Node_ComponentBoundEvent}
   > Branch(Condition=Find(
     TargetMap:Map of Names to Gameplay Tag Structures=Get PanelTags() -> (PanelTags:Map of Names to Gameplay Tag Structures)
     Key:"Name (by ref)"=@n6.TabId
    ) -> (Value:Gameplay Tag Structure).ReturnValue)
   > n7 = Branch(Condition="Equal (Name)"(A=@n6.TabId, B=Settings))
   > RequestCloseSidebar(self=Get FortSidebarVM())
   > RequestOpenSettings!(
    execute
    self:Fort Settings Context Object Reference=Get FortSettingsContext!() -> (ReturnValue:Fort Settings Context Object Reference) {is=title K2Node_GetBlueprintContextEx}
    ContextWidget:Common User Widget Object Reference=Self-Reference()
    SettingsTabId:Name=None
    SettingToSelect:String
   ) -> (then)
   n8 = Get FortSidebarVM()
   Branch(
    execute=@n7.else
    Condition="Not Equal (Name)"(A=Get EnabledFeatureId!(self:Fort Sidebar VM Object Reference=@n8) -> (EnabledFeatureId:Name=None), B=@n6.TabId) -> (ReturnValue=~)
   )
   > SetFeature!(execute, self:Fort Sidebar VM Object Reference=Get FortSidebarVM(), NewFeatureId:"Name (by ref)"=@n6.TabId) -> (then)
   > Set isNewFeatureSelected(isNewFeatureSelected=true)
   n9 = OnFeatureChanged!() -> (OutputDelegate, then, FeatureId:Name) {is=title K2Node_CustomEvent}
   n10 = Find(TargetMap=Get PanelTags(), Key=@n9.FeatureId)
   > n11 = Branch(Condition=AND Boolean!(A=IsActivated!(self:Common Activatable Widget Object Reference) -> (ReturnValue:bool), B=@n10.ReturnValue) -> (ReturnValue))
   > ClearNavigationStack(self:Fort Sidebar Navigation VM Object Reference=Get FortSidebarNavigationVM() -> (FortSidebarNavigationVM:Fort Sidebar Navigation VM Object Reference))
   > Sideways Navigation!(execute:Exec, Tag:Gameplay Tag Structure=@n10) -> (then:Exec) {is=title K2Node_MacroInstance}
   > ST Dispatch Transition!(execute:Exec, TransitionTag:Gameplay Tag Structure=(TagName="UI.StateTree.TransitionTag.Sidebar.Overlay.Closed")) -> (then:Exec) {is=title K2Node_MacroInstance}
   > SelectTabByID!(
    execute
    self:Common Tab List Widget Base Object Reference=Get MainTabGroup() -> (MainTabGroup:WBP UIKit Tab Group Object Reference)
    TabNameID:Name=@n9.FeatureId
    bSuppressClickFeedback:bool=true
   ) -> (then, ReturnValue:bool)
   Set isNewFeatureSelected(execute=@n11.else, isNewFeatureSelected=false)
   n12 = OnDeepLinkRequest!() -> (OutputDelegate, then, DeepLinkRequestContext:Fort Sidebar Deep Link Request Context Structure) {is=title K2Node_CustomEvent}
   > DelayUntilNextFrame!(execute) -> (then)
   n13 = Break Fort Sidebar Deep Link Request Context!(FortSidebarDeepLinkRequestContext:"Fort Sidebar Deep Link Request Context Structure (by ref)"=@n12.DeepLinkRequestContext) -> (
    bRequestPending:bool=~
    NavigationTag:Gameplay Tag Structure
    ConversationId:String
    SocialUser:Fort Social User Object Reference
    WidgetName:Name
   ) {is=title K2Node_BreakStruct}
   > Branch(Condition=AND Boolean(A=IsActivated(), B=@n13))
   > n14 = Construct Object from Class!(
    execute
    Class=/Sidebar/Shared/Navigation/BP_NavigationPoint.BP_NavigationPoint_C
    self
    SocialUser:Fort Social User Object Reference=@n13.SocialUser
    ConversationId:Name=String To Name(InString=@n13.ConversationId)
    ConfirmationVM:Sidebar Confirmation Panel VM Object Reference
    ReservedSpots:int=0
    SocialItemVM:Fort Social Item VM Object Reference
    NavigationTag:Gameplay Tag Structure=@n13.NavigationTag
   ) -> (then, ReturnValue:BP Navigation Point Object Reference)
   > PushNavigation!(
    execute
    self:Fort Sidebar Navigation VM Object Reference=Get FortSidebarNavigationVM()
    NavigationPoint:Fort Sidebar Navigation Point Object Reference=@n14
   ) -> (then)
   > ResetDeepLinkRequest(self:Fort Sidebar VM Object Reference=Get FortSidebarVM())
   n15 = OnLocalPlayerMuted_Event() {is=title K2Node_CustomEvent}
   > PlaySound2D(Sound=/Sidebar/Audio/MSS_UI_PTT_DeActivate)
   n16 = OnLocalPlayerUnmuted_Event() {is=title K2Node_CustomEvent}
   > PlaySound2D(Sound=/Sidebar/Audio/MSS_UI_PTT_Activate)
   n17 = InitLocalAudioCallbacks() {is=title K2Node_CustomEvent}
   n18 = OnPushToTalkBegan_Event() {is=title K2Node_CustomEvent}
   n19 = Get FortAudioLocalPlayerSubsystem() -> (ReturnValue:Fort Audio Local Player Subsystem Object Reference) {is=title K2Node_GetSubsystem}
   n20 = Assign On Push to Talk Began!(execute:Exec=@n17.then, self:Fort Audio Local Player Subsystem Object Reference=@n19, Delegate:"Delegate (by ref)"=@n18) -> (then:Exec) {is=title K2Node_AssignDelegate}
   n21 = OnPushToTalkEnded_Event() {is=title K2Node_CustomEvent}
   Assign On Push to Talk Ended!(execute:Exec=@n20.then, self:Fort Audio Local Player Subsystem Object Reference=@n19, Delegate:"Delegate (by ref)"=@n21) -> (then:Exec) {is=title K2Node_AssignDelegate}
   > Assign On Local Player Muted!(execute:Exec, self:Fort Audio Local Player Subsystem Object Reference=@n19, Delegate:"Delegate (by ref)"=@n15) -> (then:Exec) {is=title K2Node_AssignDelegate}
   > Assign On Local Player Unmuted!(execute:Exec, self:Fort Audio Local Player Subsystem Object Reference=@n19, Delegate:"Delegate (by ref)"=@n16) -> (then:Exec) {is=title K2Node_AssignDelegate}
   n22 = On Tooltip Action Clicked() {is=title K2Node_CustomEvent}
   > Unbind Event from Button Other Action Clicked(
    self:WBP UIKit Coach Mark Base Object Reference=Get WBP_UIKit_CoachMark_Cilantro!(self:WBP Sidebar Cilantro Coach Mark Object Reference=Get WBP_Sidebar_Cilantro_CoachMark() -> (WBP_Sidebar_Cilantro_CoachMark:WBP Sidebar Cilantro Coach Mark Object Reference)) -> (WBP_UIKit_CoachMark_Cilantro:WBP UIKit Coach Mark Object Reference)
    Delegate=@n22
   ) {is=title K2Node_RemoveDelegate}
   > OpenPlaytimeControlsPanel()
   n23 = OrientationChanged!() -> (OutputDelegate, then, inScreenOrientation:EScreenOrientation Enum=Unknown) {is=title K2Node_CustomEvent}
   > HandleOrientationChanged!(execute, self, ScreenOrientation:EScreenOrientation Enum=@n23.inScreenOrientation) -> (then)
   n24 = HandleOnVirtualKeyboardShown!() -> (OutputDelegate, then, RectTop:int, RectLeft:int, RectBottom:int, RectRight:int, VirtualKeyboardInputHeight:int) {is=title K2Node_CustomEvent}
   n25 = GetConfigDataBoolWithDefault!(EntryName:String=EnableSidebarChatResizeAboveKeyboard, bDefaultValue:bool=false) -> (ReturnValue:bool)
   > Branch(Condition=@n25)
   > MoveSidebarAboveVirtualKeyboard!(execute, self, RectTop:int=@n24.RectTop, RectBottom:int=@n24.RectBottom, Virtual Keyboard Input Height:int=@n24.VirtualKeyboardInputHeight) -> (then)
   > HideTabBar()
   HandleOnVirtualKeyboardHidden() {is=title K2Node_CustomEvent}
   > Branch(Condition=@n25)
   > ResetSidebarPosition()
   > ResetTabBar()
   Event Destruct()
   > Remove State Tree!(
    execute
    self:UIState Tree Manager Object Reference=Get UIStateTreeManager() -> (ReturnValue:UIState Tree Manager Object Reference) {is=title K2Node_GetSubsystem}
    Handle:UIState Tree Handle Structure=Get StateTreeHandle() -> (StateTreeHandle:UIState Tree Handle Structure)
    AnimationHints:Gameplay Tag Container Structure
    IncompatibleAnimationHints:Gameplay Tag Container Structure
   ) -> (then, ReturnValue:bool)
   > Set StateTreeHandle(StateTreeHandle:UIState Tree Handle Structure) -> (Output_Get:UIState Tree Handle Structure)
   n26 = Create Event()
   n27 = Get CommonInputSubsystem()
   > Unbind Event from On Input Method Changed(self=@n27, Delegate=@n26)
   > n28 = Cast To PlatformGameInstance!(execute, Object=GetGameInstance!() -> (ReturnValue:Game Instance Object Reference)) -> (then, CastFailed, AsPlatform Game Instance:Platform Game Instance Object Reference) {is=title K2Node_DynamicCast}
   n29 = Create Event()
   > Unbind Event from Application Received Screen Orientation Changed Notification Delegate(self:Platform Game Instance Object Reference=@n28, Delegate=@n29) {is=title K2Node_RemoveDelegate}
   n30 = Create Event()
   n31 = Get FortPlatformEventVM() -> (FortPlatformEventVM:Fort Platform Event VM Object Reference)
   > Unbind Event from On Virtual Keyboard Shown(self:Fort Platform Event VM Object Reference=@n31, Delegate=@n30) {is=title K2Node_RemoveDelegate}
   n32 = Create Event()
   n33 = Get FortPlatformEventVM()
   > Unbind Event from On Virtual Keyboard Hidden(self:Fort Platform Event VM Object Reference=@n33, Delegate=@n32) {is=title K2Node_RemoveDelegate}
   Event Construct()
   > Set Tab Group to Fill!(execute:Exec) -> (then:Exec) {is=title K2Node_MacroInstance}
   > InitializeVisuals()
   > SetCompactUIStyle() {is=title K2Node_CallFunction}
   n34 = isUseEpicParties()
   > Set ButtonClassOverride!(
    execute
    ButtonClassOverride:Map of Names to Common Button Base Class References=Select!(
     Option 0:Map of Names to Common Button Base Class References=Get Main Tab Button Class Override LocalPlayerTab() -> (Main Tab Button Class Override LocalPlayerTab:Map of Names to Common Button Base Class References)
     Option 1:Map of Names to Common Button Base Class References=Get Main Tab Button Class Override Social() -> (Main Tab Button Class Override Social:Map of Names to Common Button Base Class References)
     Index:bool=@n34
    ) -> (ReturnValue:Map of Names to Common Button Base Class References)
    self:WBP UIKit Tab Group Base Object Reference=Get MainTabGroup()
   ) -> (then, Output_Get:Map of Names to Common Button Base Class References)
   > InitializeSidebarVM!(
    execute
    self:Fort Sidebar VM Object Reference=Get FortSidebarVM()
    FeaturesConfig:Fort Sidebar Feature Config Object Reference=Select(
     Option 0:Fort Sidebar Feature Config Object Reference=/Sidebar/Panels/SidebarFeatureConfig
     Option 1:Fort Sidebar Feature Config Object Reference=/Sidebar/Panels/SidebarFeatureConfig_V2
     Index=@n34
    ) -> (ReturnValue:Fort Sidebar Feature Config Object Reference)
   ) -> (then)
   > Bind Event to On Input Method Changed(self=@n27, Delegate=@n26)
   > OnInputMethodChanged!(execute, self, Current Input Type:ECommonInputType Enum=GetCurrentInputType(self=@n27)) -> (then)
   n35 = Get PanelsSlottingHost()
   > Set InitialSlottingHostPadding(InitialSlottingHostPadding:Margin Structure=Get Padding!(self:Grid Slot Object Reference=SlotAsGridSlot(Widget=@n35)) -> (Padding:Margin Structure)) -> (Output_Get:Margin Structure)
   > n36 = Cast To PlatformGameInstance(Object=GetGameInstance())
   > Assign Application Received Screen Orientation Changed Notification Delegate!(execute:Exec, self:Platform Game Instance Object Reference=@n36, Delegate:"Delegate (by ref)"=@n29) -> (then:Exec) {is=title K2Node_AssignDelegate}
   > HandleOrientationChanged(ScreenOrientation=GetDeviceOrientation!() -> (ReturnValue:EScreenOrientation Enum=Unknown))
   > Bind Event to On Virtual Keyboard Shown(self:Fort Platform Event VM Object Reference=@n31, Delegate=@n30) {is=title K2Node_AddDelegate}
   > Bind Event to On Virtual Keyboard Hidden(self:Fort Platform Event VM Object Reference=@n33, Delegate=@n32) {is=title K2Node_AddDelegate}
   n37 = Event PreConstruct()
   > Branch(Condition=@n37.IsDesignTime)
   > Set Tab Group to Fill()
   > InitializeVisuals()
   Event On Activated()
   n38 = Get EvoPersistentUISubsystem() -> (ReturnValue:Evo Persistent UISubsystem Object Reference) {is=title K2Node_GetSubsystem}
   > EnablePersistentUI(self:Evo Persistent UISubsystem Object Reference=@n38)
   > OnSidebarOpen!(execute, self:Fort Sidebar VM Object Reference=Get FortSidebarVM(), Sidebar:Widget Object Reference=Self-Reference()) -> (then)
   > ClearNavigationStack(self=Get FortSidebarNavigationVM())
   n39 = Get EnabledFeatureId(self=Get FortSidebarVM()) -> (EnabledFeatureId=None)
   > Sideways Navigation(Tag=Find(TargetMap=Get PanelTags(), Key=@n39))
   > SelectTabByID(self=Get MainTabGroup(), TabNameID=@n39, bSuppressClickFeedback=true)
   > Fire Event Main View Opened!(execute, View:Widget Object Reference=Self-Reference(), AdditionalInfo:Flow Analytics Additional Info Structure, bIsOverlay:bool=true) -> (then)
   > OnDeepLinkRequest!(
    execute
    self
    DeepLinkRequestContext:Fort Sidebar Deep Link Request Context Structure=Get DeepLinkRequestContext!(self:Fort Sidebar VM Object Reference=Get FortSidebarVM()) -> (DeepLinkRequestContext:Fort Sidebar Deep Link Request Context Structure)
   ) -> (then) {is=title K2Node_CallFunction}
   > n40 = Branch(Condition=ShouldShowOnboarding!(self:Fort Sidebar Onboard Tooltip Widget Object Reference=Get WBP_Sidebar_Cilantro_CoachMark()) -> (ReturnValue:bool))
   n41 = Get SpacerForCoachmarks() -> (SpacerForCoachmarks:Spacer Object Reference)
   > n42 = SetVisibility(self=@n41, InVisibility=SelfHitTestInvisible)
   HandleOrientationChanged(execute=[@n42.then, @n47.then], ScreenOrientation=GetDeviceOrientation() -> (ReturnValue=Unknown))
   > Branch(Condition=ShouldShowOnboarding(self=Get WBP_Sidebar_Cilantro_CoachMark()))
   > Bind Event to Button Other Action Clicked(self:WBP UIKit Coach Mark Base Object Reference=Get WBP_UIKit_CoachMark_Cilantro(self=Get WBP_Sidebar_Cilantro_CoachMark()), Delegate=@n22) {is=title K2Node_AddDelegate}
   > n43 = Branch(Condition=isUseEpicParties())
   n44 = Get Current Input Type()
   > SetGridSlotNudge!(
    execute
    Widget:Widget Object Reference=Get FortVisualAttachment() -> (FortVisualAttachment:Fort Visual Attachment Object Reference)
    Nudge:Vector 2D Structure=Select!(
     MouseAndKeyboard:Vector 2D Structure=(X=-15,Y=75)
     Gamepad:Vector 2D Structure=(X=-185,Y=75)
     Touch:Vector 2D Structure=(X=-15,Y=75)
     Count:Vector 2D Structure=(X=-15,Y=75)
     Index:ECommonInputType Enum=@n44
    ) -> (ReturnValue:Vector 2D Structure)
   ) -> (then)
   > n45 = SetArrowPosition!(
    execute
    self:WBP UIKit Coach Mark Base Object Reference=Get WBP_UIKit_CoachMark_Cilantro(self=Get WBP_Sidebar_Cilantro_CoachMark())
    ArrowPosition:E_UI_OutBoardPosition Enum=Select!(
     Option 0:E_UI_OutBoardPosition Enum=NewEnumerator0
     Option 1:E_UI_OutBoardPosition Enum=NewEnumerator1
     Index:bool="Equal (Enum)"(A=@n44, B=Gamepad)
    ) -> (ReturnValue:E_UI_OutBoardPosition Enum)
   ) -> (then)
   ActivateWidget(execute=[@n45.then, @n46.then], self=Get WBP_Sidebar_Cilantro_CoachMark())
   SetGridSlotNudge(
    execute=@n43.else
    Widget=Get FortVisualAttachment()
    Nudge=Select!(
     MouseAndKeyboard:Vector 2D Structure=(X=-75,Y=75)
     Gamepad:Vector 2D Structure=(X=-125,Y=75)
     Touch:Vector 2D Structure=(X=-75,Y=75)
     Count:Vector 2D Structure=(X=-75,Y=75)
     Index:ECommonInputType Enum=Get Current Input Type()
    ) -> (ReturnValue:Vector 2D Structure)
   )
   > n46 = SetArrowPosition(self=Get WBP_UIKit_CoachMark_Cilantro(self=Get WBP_Sidebar_Cilantro_CoachMark()), ArrowPosition=NewEnumerator2)
   n47 = SetVisibility(execute=@n40.else, self=@n41, InVisibility=Collapsed)
   Event OnInitialized()
   > SetVisibility(self=Self-Reference(), InVisibility=Collapsed)
   > n48 = AddStateTree!(
    execute
    self:UIState Tree Manager Object Reference=Get UIStateTreeManager()
    UIStateTree:State Tree Object Reference=/Sidebar/StateTree/ST_Sidebar
    StateTreeOwner:Player Controller Object Reference=GetOwningPlayer()
    AnimationHints:Gameplay Tag Container Structure
   ) -> (then, ReturnValue:UIState Tree Handle Structure)
   > Set StateTreeHandle(StateTreeHandle=@n48)
   > InitLocalAudioCallbacks() {is=title K2Node_CallFunction}
   Event On Deactivated()
   > DisablePersistentUI(self:Evo Persistent UISubsystem Object Reference=@n38)
   > n49 = Branch(Condition=IsActivated(self=Get WBP_Sidebar_Cilantro_CoachMark()))
   > n50 = DeactivateWidget(self=Get WBP_Sidebar_Cilantro_CoachMark())
   ResetToDefaultFeature(execute=[@n50.then, @n49.else], self:Fort Sidebar VM Object Reference=Get FortSidebarVM())
   > ClearNavigationStack(self=Get FortSidebarNavigationVM())
   > DeselectAll(self:UIKit Tab Group Object Reference=Get MainTabGroup())
   > OnSidebarClose!(execute, self:Fort Sidebar VM Object Reference=Get FortSidebarVM(), Sidebar:Widget Object Reference=Self-Reference()) -> (then)
   > ST Dispatch Transition(TransitionTag=(TagName="UI.StateTree.TransitionTag.Sidebar.Overlay.Closed"))
   > ST Dispatch Transition(TransitionTag=(TagName="UI.StateTree.TransitionTag.Sidebar.Close"))
   > SetVisibility(self=Get SpacerForCoachmarks(), InVisibility=Collapsed)
   PlaySound2D(execute=@n21.then, Sound=&MSS_UI_PTT_DeActivate)
   PlaySound2D(execute=@n18.then, Sound=&MSS_UI_PTT_Activate)
  ]}
 HandleOrientationChanged {nodes [
   HandleOrientationChanged!() -> (then, ScreenOrientation:EScreenOrientation Enum=Unknown) {is=title K2Node_FunctionEntry}
   > n1 = GetConsoleVariableBoolValue!(execute, VariableName:String=PilgrimInverse.UsePortraitMode) -> (then, ReturnValue:bool)
   > Branch(Condition=AND Boolean(A=@n1, B=IsUsingCompactUI()))
   n2 = DisplayOrientationFromScreenOrientation!(ScreenOrientation:EScreenOrientation Enum=Get ScreenOrientation() -> (ScreenOrientation:EScreenOrientation Enum=Unknown)) -> (DisplayOrientation:EDisplayOrientationType Enum=Landscape)
   > Branch(Condition="Not Equal (Enum)"!(A:EDisplayOrientationType Enum=Get CachedDisplayOrientation() -> (CachedDisplayOrientation=Landscape), B:EDisplayOrientationType Enum=@n2) -> (ReturnValue:bool))
   > n3 = Set CachedDisplayOrientation(CachedDisplayOrientation:EDisplayOrientationType Enum=@n2) -> (Output_Get:EDisplayOrientationType Enum=Landscape)
   > n4 = Branch(Condition="Equal (Enum)"(A:EDisplayOrientationType Enum=@n3, B:EDisplayOrientationType Enum=Portrait))
   > n5 = Branch(Condition=CompareToScreenAspectRatio!(Comparison:EAspectRatioComparison Enum=GreaterThan, AspectRatio:EAspectRatio Enum=Ratio_23_16) -> (ReturnValue:bool))
   > UpdateLayoutForOrientation!(execute, self, NewDisplayOrientation:EDisplayOrientationType Enum=Portrait) -> (then)
   UpdateLayoutForOrientation(execute=[@n5.else, @n4.else], NewDisplayOrientation=Landscape)
  ]}
 HideTabBar {nodes [
   HideTabBar() {is=title K2Node_FunctionEntry}
   > Branch(Condition=GetConfigDataBoolWithDefault(EntryName=EnableHideTabBarWhenVirtualKeyboardOpen, bDefaultValue=false))
   > SetVisibility(self=Get TabGroupContainer() -> (TabGroupContainer:Grid Panel Object Reference), InVisibility=Collapsed)
   > SetPadding(self=SlotAsGridSlot(Widget=Get PanelsSlottingHost()), InPadding_Left=0, InPadding_Top=0, InPadding_Right=0, InPadding_Bottom=0)
  ]}
 InitializeVisuals {nodes [
   InitializeVisuals() {is=title K2Node_FunctionEntry}
   n1 = Get Backplate() -> (Backplate:WBP Sidebar Block Backplate Object Reference)
   > SetBackplateColor!(execute, self:WBP Sidebar Block Backplate Object Reference=@n1, Backplate Switch:E_UI_Sidebar_Button_Backplate_Color Enum=NewEnumerator11) -> (then)
   > SetCornerRadius!(execute, self:WBP Sidebar Block Backplate Object Reference=@n1, Radius In Pixels:bool=true, Corner Radius:double=36) -> (then)
   n2 = Get TabBackplate() -> (TabBackplate:WBP Sidebar Block Backplate Object Reference)
   > SetBackplateColor(self=@n2, Backplate Switch=NewEnumerator2)
   > SetCornerRadius(self=@n2, Radius In Pixels=true, Corner Radius=24)
  ]}
 MoveSidebarAboveVirtualKeyboard {nodes [
   MoveSidebarAboveVirtualKeyboard!() -> (then, RectTop:int, RectBottom:int, Virtual Keyboard Input Height:int) {is=title K2Node_FunctionEntry}
   n1 = Get FortSidebarVM()
   > Branch(Condition="Equal (Name)"(A=Get EnabledFeatureId(self=@n1) -> (EnabledFeatureId=None), B=Chat))
   > n2 = Branch(Condition=GetConfigDataBoolWithDefault(EntryName=UseFortFunctionLibraryCalculateChatPadding, bDefaultValue=true))
   n3 = Get SizeBox_Sidebar() -> (SizeBox_Sidebar:Size Box Object Reference)
   > SetPadding!(
    execute
    self=SlotAsGridSlot(Widget=@n3)
    InPadding:Margin Structure=CalculateChatPadding!(
     TargetWidget:Widget Object Reference=@n3
     RectTop:int=Get RectTop() -> (RectTop:int=0)
     RectBottom:int=Get RectBottom() -> (RectBottom:int=0)
     VirtualKeyboardInputHeight:int=Get Virtual Keyboard Input Height() -> (Virtual Keyboard Input Height:int=0)
    ) -> (ReturnValue:Margin Structure)
   ) -> (then)
   Set Scale(
    execute=@n2.else
    Scale:double=float / float(A:int=Get RectBottom() -> (RectBottom=0), B=GetViewportSize!() -> (ReturnValue_X:double, ReturnValue_Y:double).ReturnValue_Y)
   ) -> (Output_Get:double)
   n4 = GetTickSpaceGeometry!(self:Widget Object Reference=Get SizeBox_Sidebar()) -> (ReturnValue:"Geometry Structure (by ref)")
   n5 = int - int(A=Get RectBottom() -> (RectBottom=0), B=Get RectTop() -> (RectTop=0))
   n6 = Get Virtual Keyboard Input Height() -> (Virtual Keyboard Input Height=0)
   n7 = GetTickSpaceGeometry()
   n8 = vector2d + vector2d(
    A:Vector 2D Structure=GetLocalTopLeft!(Geometry:"Geometry Structure (by ref)"=@n7) -> (ReturnValue:Vector 2D Structure)
    B:Vector 2D Structure=GetLocalSize!(Geometry:"Geometry Structure (by ref)"=@n7) -> (ReturnValue:Vector 2D Structure)
   ) -> (ReturnValue:Vector 2D Structure) {is=title K2Node_PromotableOperator}
   n9 = Get RectBottom() -> (RectBottom=0)
   n10 = Get Scale() -> (Scale:double)
   n11 = float - float(
    A=float / float(A=int + int(A=@n5, B=@n6), B=Get Scale())
    B=float - float(
     A=float / float(A=@n9, B=@n10)
     B="Local to Absolute (Coordinates)"!(Geometry:"Geometry Structure (by ref)"=@n7, LocalCoordinate:Vector 2D Structure=@n8) -> (ReturnValue_X:double, ReturnValue_Y:double).ReturnValue_Y
    )
   )
   > SetPadding!(
    execute
    self=SlotAsGridSlot(Widget=Get SizeBox_Sidebar())
    InPadding_Left:float=0
    InPadding_Top:float=0
    InPadding_Right:float=0
    InPadding_Bottom:float=float + float(
     A="Absolute to Local (Coordinates)"!(Geometry:"Geometry Structure (by ref)"=@n4, AbsoluteCoordinate_X:double=0, AbsoluteCoordinate_Y:double=@n11) -> (ReturnValue_X:double, ReturnValue_Y:double).ReturnValue_Y
     B=GetConfigDataFloatWithDefault!(EntryName:String=SidebarTextChatAdditionalPaddingWithVKOn, DefaultValue:float=0) -> (ReturnValue:float)
    )
   ) -> (then)
  ]}
 NavigationToTabGroup {nodes [
   NavigationToTabGroup!() -> (then, Navigation:EUINavigation Enum) {is=title K2Node_FunctionEntry}
   n1 = Get FortSidebarVM()
   > Return Node(ReturnValue=GetTabButtonBaseByID!(
     self:Common Tab List Widget Base Object Reference=Get MainTabGroup()
     TabNameID:Name=Get EnabledFeatureId(self=@n1) -> (EnabledFeatureId=None)
    ) -> (ReturnValue:Common Button Base Object Reference))
  ]}
 OnInputMethodChanged {nodes [
   n1 = FindParentWidgetOfType!(execute, StartingWidget:Widget Object Reference=Self-Reference(), Type:Widget Class Reference=&/WBP_Sidebar.WBP_Sidebar_C) -> (then, ReturnValue:WBP Sidebar Object Reference)
   > SetVisibility(self=Get SpacerForCoachmarks!(self:WBP Sidebar Object Reference=@n1) -> (SpacerForCoachmarks), InVisibility=SelfHitTestInvisible)
   n2 = OnInputMethodChanged!() -> (then, Current Input Type:ECommonInputType Enum) {is=title K2Node_FunctionEntry}
   > Set Current Input Type(Current Input Type:ECommonInputType Enum=@n2) -> (Output_Get:ECommonInputType Enum)
   n3 = Get FortSidebarVM()
   > SetupAllTabs!(
    execute
    self
    Data:Array of Fort Sidebar Feature VM Object References=Get Features!(self:Fort Sidebar VM Object Reference=@n3) -> (Features:Array of Fort Sidebar Feature VM Object References)
   ) -> (then)
   > SelectTabByID(self=Get MainTabGroup(), TabNameID=Get EnabledFeatureId(self=@n3) -> (EnabledFeatureId=None), bSuppressClickFeedback=true)
   > Branch(Condition=ShouldShowOnboarding(self=Get WBP_Sidebar_Cilantro_CoachMark()))
   > n4 = Branch(Condition=isUseEpicParties())
   n5 = Get Current Input Type()
   > SetGridSlotNudge(
    Widget=Get FortVisualAttachment()
    Nudge=Select(MouseAndKeyboard=(X=-15,Y=75), Gamepad=(X=-185,Y=75), Touch=(X=-15,Y=75), Count=(X=-15,Y=75), Index=@n5)
   )
   > SetArrowPosition(
    self=Get WBP_UIKit_CoachMark_Cilantro(self=Get WBP_Sidebar_Cilantro_CoachMark())
    ArrowPosition=Select!(
     Option 0:E_UI_OutBoardPosition Enum=NewEnumerator0
     Option 1:E_UI_OutBoardPosition Enum=NewEnumerator1
     Index:bool="Equal (Enum)"(A:ECommonInputType Enum=@n5, B:ECommonInputType Enum=Gamepad)
    ) -> (ReturnValue:E_UI_OutBoardPosition Enum)
   )
   SetGridSlotNudge(
    execute=@n4.else
    Widget=Get FortVisualAttachment()
    Nudge=Select!(
     MouseAndKeyboard:Vector 2D Structure=(X=-75,Y=75)
     Gamepad:Vector 2D Structure=(X=-125,Y=75)
     Touch:Vector 2D Structure=(X=-75,Y=75)
     Count:Vector 2D Structure=(X=-75,Y=75)
     Index:ECommonInputType Enum=Get Current Input Type()
    ) -> (ReturnValue:Vector 2D Structure)
   )
   > SetArrowPosition(self=Get WBP_UIKit_CoachMark_Cilantro(self=Get WBP_Sidebar_Cilantro_CoachMark()), ArrowPosition=NewEnumerator2)
  ]}
 OnSidebarRootNavigationEnter {nodes [
   n1 = OnSidebarRootNavigationEnter!() -> (then, Type:EUINavigation Enum, OldFocusedWidget:Widget Object Reference, NewFocusedWidget:Widget Object Reference) {is=title K2Node_FunctionEntry}
   n2 = GetOuterUserWidget!(Widget:Widget Object Reference=@n1.NewFocusedWidget) -> (ReturnValue:User Widget Object Reference)
   > n3 = Cast To WBP_Sidebar_Button_Base!(execute, Object=@n2) -> (then, CastFailed, AsWBP Sidebar Button Base:WBP Sidebar Button Base Object Reference) {is=title K2Node_DynamicCast}
   > Set LastNavigationDirection!(
    execute
    LastNavigationDirection:EUINavigation Enum=Get Type() -> (Type:EUINavigation Enum)
    self:WBP Sidebar Block Background Button Object Reference=Get Background!(self:WBP Sidebar Button Base Object Reference=@n3) -> (Background:WBP Sidebar Block Background Button Object Reference)
   ) -> (then, Output_Get:EUINavigation Enum)
   n4 = Cast To WBP_Sidebar_ButtonGeneric!(execute=@n3.CastFailed, Object=@n2) -> (then, CastFailed, AsWBP Sidebar Button Generic:WBP Sidebar Button Generic Object Reference) {is=title K2Node_DynamicCast}
   > Set LastNavigationDirection(
    LastNavigationDirection=Get Type()
    self=Get ButtonBackground!(self:WBP Sidebar Button Generic Object Reference=@n4) -> (ButtonBackground:WBP Sidebar Block Background Button Object Reference)
   )
  ]}
 OpenPlaytimeControlsPanel {nodes [
   OpenPlaytimeControlsPanel() {is=title K2Node_FunctionEntry}
   > SetFeature(self=Get FortSidebarVM(), NewFeatureId=MakeLiteralName(Value=MainMenu))
   > n1 = Construct Object from Class(
    Class=&BP_NavigationPoint_C
    ConversationId=None
    ReservedSpots=0
    NavigationTag=(TagName="UI.StateTree.TransitionTag.Sidebar.Settings.TimeLimits")
   )
   > PushNavigation(self=Get FortSidebarNavigationVM(), NavigationPoint=@n1)
  ]}
 ResetSidebarPosition {nodes [
   ResetSidebarPosition() {is=title K2Node_FunctionEntry}
   > SetPadding!(execute, self=SlotAsGridSlot(Widget=Get SizeBox_Sidebar()), InPadding:Margin Structure) -> (then)
  ]}
 ResetTabBar {nodes [
   ResetTabBar() {is=title K2Node_FunctionEntry}
   > Branch(Condition=GetConfigDataBoolWithDefault(EntryName=EnableHideTabBarWhenVirtualKeyboardOpen, bDefaultValue=false))
   > SetVisibility(self=Get TabGroupContainer(), InVisibility=SelfHitTestInvisible)
   > SetPadding(
    self=SlotAsGridSlot(Widget=Get PanelsSlottingHost())
    InPadding=Get InitialSlottingHostPadding() -> (InitialSlottingHostPadding:Margin Structure)
   )
  ]}
 ST_DispatchTransition {nodes [
   n1 = Inputs!() -> (execute, TransitionTag:Gameplay Tag Structure)
   > Dispatch Transition Event!(
    execute
    self:UIState Tree Manager Object Reference=Get UIStateTreeManager()
    TransitionTag:Gameplay Tag Structure=@n1
    StateTreeOwner:Player Controller Object Reference=GetOwningPlayer()
    AnimationHints:Gameplay Tag Container Structure
    IncompatibleAnimationHints:Gameplay Tag Container Structure
   ) -> (then)
   > Outputs!(then:Exec)
  ]}
 SetCompactUIStyle {nodes [
   SetCompactUIStyle() {is=title K2Node_FunctionEntry}
   > Branch(Condition=IsUsingCompactUI())
   > Set SidebarWidth(SidebarWidth:double=768) -> (Output_Get:double)
   > EvaluateSafezones() {is=title K2Node_CallFunction}
   n1 = Get MainTabGroup()
   > Set UseMinButtonSize!(execute, UseMinButtonSize:bool=false, self:WBP UIKit Tab Group Base Object Reference=@n1) -> (then, Output_Get:bool)
   > Set IconSize!(execute, IconSize:Vector 2D Structure=(X=54,Y=54), self:WBP UIKit Tab Group Base Object Reference=@n1) -> (then, Output_Get:Vector 2D Structure)
   n2 = SlotAsGridSlot(Widget=Get PanelsSlottingHost())
   n3 = Get!(self:Grid Slot Object Reference=@n2) -> (Padding_Left:float=~, Padding_Top:float=~, Padding_Right:float=~, Padding_Bottom:float=~)
   > SetPadding!(
    execute
    self=@n2
    InPadding_Left:float=@n3
    InPadding_Top:float=120
    InPadding_Right:float=@n3.Padding_Right
    InPadding_Bottom:float=@n3.Padding_Bottom
   ) -> (then)
   > SetWidthOverride(self=Get SizeBox_Sidebar(), InWidthOverride=Get SidebarWidth() -> (SidebarWidth:double))
  ]}
 SetTabGroupToFill {nodes [
   Inputs!() -> (execute)
   n1 = Get MainTabGroup()
   > n2 = Cast To ScrollBoxSlot!(
    execute
    Object=Get Slot!(self:Widget Object Reference=Get TabsContainer!(self:WBP UIKit Tab Group Base Object Reference=@n1) -> (TabsContainer:Grid Panel Object Reference)) -> (Slot:Panel Slot Object Reference)
   ) -> (then, CastFailed, AsScroll Box Slot:Scroll Box Slot Object Reference) {is=title K2Node_DynamicCast}
   > Set Size!(
    execute
    Size:Slate Child Size Structure=Make Slate Child Size!(Value:float=1, SizeRule:ESlateSizeRule Enum=Fill) -> (SlateChildSize:Slate Child Size Structure) {is=title K2Node_MakeStruct}
    self:Scroll Box Slot Object Reference=@n2
   ) -> (then, Output_Get:Slate Child Size Structure)
   > Outputs()
  ]}
 SetupAllTabs {nodes [
   n1 = SetupAllTabs!() -> (then, Data:Array of Fort Sidebar Feature VM Object References) {is=title K2Node_FunctionEntry}
   > RemoveAllTabs(self:Common Tab List Widget Base Object Reference=Get MainTabGroup())
   > Clear!(execute, TargetMap:Map of Names to Gameplay Tag Structures=Get PanelTags()) -> (then) {is=title K2Node_CallFunction}
   > Set CurrentInputType(CurrentInputType:ECommonInputType Enum=GetCurrentInputType(self=Get CommonInputSubsystem())) -> (Output_Get:ECommonInputType Enum)
   > n2 = For Each Loop(Array:Array of Fort Sidebar Feature VM Object References=@n1) -> (Array Element:Fort Sidebar Feature VM Object Reference)
   n3 = Get TabButtonInputTypeRestrictions!(self:Fort Sidebar Feature VM Object Reference=@n2) -> (TabButtonInputTypeRestrictions:Array of ECommonInputType Enums=MouseAndKeyboard)
   n4 = Get CurrentInputType() -> (CurrentInputType:ECommonInputType Enum)
   > Branch(Condition="Equal (Integer)"(A=Find Item!(TargetArray:Array of ECommonInputType Enums=@n3, ItemToFind:"ECommonInputType Enum (by ref)"=@n4) -> (ReturnValue), B=-1))
   n5 = Get FeatureId!(self:Fort Sidebar Feature VM Object Reference=@n2) -> (FeatureId:Name=None)
   > Add!(
    execute
    TargetMap:Map of Names to Gameplay Tag Structures=Get PanelTags()
    Key:"Name (by ref)"=@n5
    Value:"Gameplay Tag Structure (by ref)"=Get NavigationTag!(self:Fort Sidebar Feature VM Object Reference=@n2) -> (NavigationTag:Gameplay Tag Structure)
   ) -> (then) {is=title K2Node_CallFunction}
   > AddTab!(
    execute
    self:WBP UIKit Tab Group Base Object Reference=Get MainTabGroup()
    Tab VM:MVVM Base Viewmodel Object Reference=@n2
    Tab ID:Name=@n5
    Index:int=-1
   ) -> (then)
  ]}
 SidewaysNavigation {nodes [
   n1 = Inputs!() -> (execute, Tag:Gameplay Tag Structure)
   n2 = isWebSOEnabled!() -> (Web SO Enabled:bool)
   n3 = Get EnabledFeatureId(self=Get FortSidebarVM()) -> (EnabledFeatureId=None)
   n4 = OR Boolean!(A="Equal (Name)"(A=@n3, B=Social), B="Equal (Name)"(A=@n3, B=Chat), C:bool="Equal (Name)"(A=@n3, B=AddFriend)) -> (ReturnValue)
   > n5 = Construct Object from Class(
    Class=&BP_NavigationPoint_C
    ConversationId=None
    ReservedSpots=0
    NavigationTag=Select!(
     Option 0:Gameplay Tag Structure=@n1
     Option 1:Gameplay Tag Structure=(TagName="UI.StateTree.TransitionTag.Sidebar.WebSO")
     Index:bool=AND Boolean(A=@n2, B=@n4)
    ) -> (ReturnValue:Gameplay Tag Structure)
   )
   > SidewaysNavigation!(
    execute
    self:Fort Sidebar Navigation VM Object Reference=Get FortSidebarNavigationVM()
    NavigationPoint:Fort Sidebar Navigation Point Object Reference=@n5
   ) -> (then)
   > Outputs()
  ]}
 UpdateLayoutForOrientation {nodes [
   n1 = UpdateLayoutForOrientation!() -> (then, NewDisplayOrientation:EDisplayOrientationType Enum=Landscape) {is=title K2Node_FunctionEntry}
   n2 = Get WBP_Sidebar_Cilantro_CoachMark()
   > n3 = Branch(Condition=NOT Boolean(A=ShouldShowOnboarding(self=@n2)))
   n4 = Get SpacerForCoachmarks!() -> (SpacerForCoachmarks)
   > n5 = SetVisibility(self=@n4, InVisibility=Collapsed)
   n6 = SlotAsGridSlot(Widget=Get SizeBox_Sidebar())
   n7 = Get(self=@n6) -> (Padding_Left=~, Padding_Top=~, Padding_Right=~, Padding_Bottom=~)
   Set!(
    execute=[@n5.then, @n11.then]
    Padding_Left:float=@n7
    Padding_Top:float=Select!(Landscape:float=0, Portrait:float=280, Index:EDisplayOrientationType Enum=@n1) -> (ReturnValue:float)
    Padding_Right:float=@n7.Padding_Right
    Padding_Bottom:float=@n7.Padding_Bottom
    self:Grid Slot Object Reference=@n6
   ) -> (then, Output_Get:Margin Structure)
   n8 = Get!() -> (SafeZonePadding_Left:float=~, SafeZonePadding_Top:float=~, SafeZonePadding_Right:float=~, SafeZonePadding_Bottom:float=~)
   > Set Grid Slot Padding!(
    execute:Exec
    Widget:Widget Object Reference=Get SafeZone()
    Padding_Left:float=Select(Landscape=0, Portrait=@n8, Index=@n1)
    Padding_Top:float=@n8.SafeZonePadding_Top
    Padding_Right:float=@n8.SafeZonePadding_Right
    Padding_Bottom:float=@n8.SafeZonePadding_Bottom
   ) -> (then:Exec) {is=title K2Node_MacroInstance}
   > EvaluateSafezones()
   > n9 = Switch on EDisplayOrientationType!(execute, Selection:EDisplayOrientationType Enum=@n1) -> (Landscape:Exec, Portrait:Exec) {is=title K2Node_SwitchEnum}
   n10 = Get SizeBox_Sidebar()
   > SetWidthOverride(self=@n10, InWidthOverride=Get SidebarWidth())
   ClearWidthOverride(execute=@n9.Portrait, self:Size Box Object Reference=@n10)
   n11 = SetVisibility(
    execute=@n3.else
    self=@n4
    InVisibility=Select(Landscape:ESlateVisibility Enum=SelfHitTestInvisible, Portrait:ESlateVisibility Enum=Collapsed, Index=@n1) -> (ReturnValue:ESlateVisibility Enum)
   )
  ]}
 WrapLeft {nodes [
   WrapLeft!() -> (then, Navigation:EUINavigation Enum) {is=title K2Node_FunctionEntry}
   n1 = Get MainTabGroup()
   n2 = int - int(A=GetTabCount!(self:Common Tab List Widget Base Object Reference=@n1) -> (ReturnValue:int=0))
   > Return Node(ReturnValue=GetTabButtonBaseByID(self=@n1, TabNameID=GetTabIdAtIndex!(self:Common Tab List Widget Base Object Reference=@n1, Index:int=@n2) -> (ReturnValue:Name=None)))
  ]}
 WrapRight {nodes [
   WrapRight!() -> (then, Navigation:EUINavigation Enum) {is=title K2Node_FunctionEntry}
   n1 = Get MainTabGroup()
   > Return Node(ReturnValue=GetTabButtonBaseByID(self=@n1, TabNameID=GetTabIdAtIndex(self=@n1, Index=0) -> (ReturnValue=None)))
  ]}
 isWebSOEnabled {nodes [
   isWebSOEnabled() {is=title K2Node_FunctionEntry}
   > n1 = GetConsoleVariableBoolValue(VariableName=FortUIUtils.Debug.UseRealSocialOverlayInDemo)
   > Return Node!(execute, Web SO Enabled:bool=@n1)
  ]}
]
variables [
 CachedDisplayOrientation {type=EDisplayOrientationType, default=Landscape}
 Current Input Type {type=ECommonInputType, default=MouseAndKeyboard}
 InitialSlottingHostPadding {type=Margin, default=<ltrb 0 0 0 0>}
 isDisliked {type=boolean}
 isLiked {type=boolean}
 isNewFeatureSelected {type=boolean}
 Main Tab Button Class Override LocalPlayerTab {type="map<Class>", default {localPlayerTab=/Sidebar/Buttons/WBP_SidebarYouTab_Button.WBP_SidebarYouTab_Button_C}}
 Main Tab Button Class Override Social {type="map<Class>", default {social=&WBP_SidebarYouTab_Button_C}}
 PanelTags {type="map<GameplayTag>", default {}}
 Previous Feature {type=string, default=None}
 SafezonePadding {type=Margin, default=<ltrb 28 24 28 24>}
 SidebarWidth {type=number, default=632}
 StateTreeHandle {type=UIStateTreeHandle, default {handleValue=4294967295}}
]
widgets {
 info {inheritedWidgetCount=0, parentClass=&CommonActivatableWidget, rootWidgetClass=&GridPanel, widgetCount=15}
 widgets [
  Sidebar_Root {
   widgetClassPath=&GridPanel
   widget=Sidebar_Root
   properties {bOverride_Cursor=true, navigation=&self:WidgetTree.Sidebar_Root.WidgetNavigation_0, columnFill [1], rowFill [1]}
   uIComponents [{
     component=&self:UIComponentWidgetBlueprintExtension_0.ComponentContainer.NavigationUIComponent_0
     componentClassPath=/Script/UMG.NavigationUIComponent
    }]
  }
  SafeZone {
   widgetClassPath=&SafeZone
   bIsVariable=true
   parent=Sidebar_Root
   widget=SafeZone
   properties {padLeft=false}
   slot {class=GridSlot, properties {padding=<ltrb 0 24 28 24>}}
  }
  ClippingGrid {
   widgetClassPath=&GridPanel
   parent=SafeZone
   widget=ClippingGrid
   properties {clipping=ClipToBounds, columnFill [0,1], rowFill [1]}
   slot {class=SafeZoneSlot}
  }
  SpacerForCoachmarks {
   widgetClassPath=&Spacer
   bIsVariable=true
   parent=ClippingGrid
   widget=SpacerForCoachmarks
   properties {visibility=Collapsed, size=<600 1>}
   slot {class=GridSlot}
  }
  SizeBox_Sidebar {
   widgetClassPath=&SizeBox
   bIsVariable=true
   parent=ClippingGrid
   widget=SizeBox_Sidebar
   properties {bOverride_WidthOverride=true, widthOverride=632}
   slot {class=GridSlot, properties {column=1}}
  }
  Sidebar {
   widgetClassPath=&GridPanel
   parent=SizeBox_Sidebar
   widget=Sidebar
   properties {visibility=Visible, columnFill [1], rowFill [1]}
   slot {class=SizeBoxSlot}
  }
  WBP_CaptureForPostBufferUpdate {
   widgetClassPath=&WBP_CaptureForPostBufferUpdate
   bIsVariable=true
   parent=Sidebar
   slot {class=GridSlot, properties {layer=-30}}
  }
  Backplate {
   widgetClassPath=/Sidebar/Buttons/Blocks/WBP_Sidebar_Block_Backplate
   bIsVariable=true
   parent=Sidebar
   widget=Backplate
   slot {class=GridSlot, properties {layer=-20}}
  }
  TabGroupContainer {
   widgetClassPath=&GridPanel
   bIsVariable=true
   parent=Sidebar
   widget=TabGroupContainer
   properties {columnFill [1]}
   slot {class=GridSlot, properties {layer=10, verticalAlignment=VAlign_Top, padding=<ltrb 6 6 6 6>}}
  }
  MainTabGroup {
   widgetClassPath=/FortUILibrary/UIKit/Tabs/WBP_UIKit_TabGroup
   bIsVariable=true
   parent=TabGroupContainer
   widget=MainTabGroup
   properties {
    animateReservedPromptSpace=false
    bIsFocusable=true
    backplateClass=/FortUILibrary/UIKit/Backplates/WBP_UIKit_Backplate_Fallback.WBP_UIKit_Backplate_Fallback_C
    backplateCornerRadius=Interactive Component
    buttonClass=/Sidebar/Buttons/WBP_SidebarTab_Button.WBP_SidebarTab_Button_C
    buttonHorizontalAlignment=HAlign_Fill
    initialSelectedTabIndex=0
    isNaturalNavigation=true
    navigation=&self:WidgetTree.MainTabGroup.WidgetNavigation_0
    nextPromptHorizontalAlignment=HAlign_Right
    prevPromptHorizontalAlignment=HAlign_Left
    promptsBackplateClass=&WBP_UIKit_Backplate_Fallback_C
    reservedPromptSpace=0
    tabGroupHorizontalAlignment=HAlign_Fill
    tabSpacing=6
    useBackplate=false
    usePromptBackplate=false
    buttonClassOverride {social=&WBP_SidebarYouTab_Button_C}
    iconSize=<36 36>
    initialTabsData [
     {bEnableTabId=false, tabId=Profile, tabViewModel=&self:WidgetTree.MainTabGroup.UIKitSelectableButtonViewModel_1}
     {bEnableTabId=false, tabId=None, tabViewModel=&self:WidgetTree.MainTabGroup.UIKitSelectableButtonViewModel_4}
     {bEnableTabId=false, tabId=None, tabViewModel=&self:WidgetTree.MainTabGroup.UIKitSelectableButtonViewModel_0}
     {bEnableTabId=false, tabId=None, tabViewModel=&self:WidgetTree.MainTabGroup.UIKitSelectableButtonViewModel_2}
     {bEnableTabId=false, tabId=None, tabViewModel=&self:WidgetTree.MainTabGroup.UIKitSelectableButtonViewModel_3}
    ]
    minButtonSize=<0 66>
    nextPromptPadding=<ltrb 0 0 24 0>
    prevPromptPadding=<ltrb 24 0 0 0>
    tabGroupPadding=<ltrb 12 12 12 12>
   }
   slot {class=GridSlot, properties {layer=1}}
  }
  TabBackplate {
   widgetClassPath=&WBP_Sidebar_Block_Backplate
   bIsVariable=true
   parent=TabGroupContainer
   widget=TabBackplate
   slot {class=GridSlot, properties {columnSpan=1, layer=-10, padding=<ltrb 6 6 6 6>}}
  }
  FortVisualAttachment {
   widgetClassPath=&FortVisualAttachment
   bIsVariable=true
   parent=TabGroupContainer
   widget=FortVisualAttachment
   properties {contentAnchor=<1 0>}
   slot {class=GridSlot, properties {horizontalAlignment=HAlign_Right, layer=99, verticalAlignment=VAlign_Top, nudge=<-75 75>}}
  }
  WBP_Sidebar_Cilantro_CoachMark {
   widgetClassPath=/Sidebar/Coachmarks/WBP_Sidebar_Cilantro_CoachMark
   bIsVariable=true
   parent=FortVisualAttachment
   widget=WBP_Sidebar_Cilantro_CoachMark
   properties {bIsFocusable=true, navigation=&self:WidgetTree.WBP_Sidebar_Cilantro_CoachMark.WidgetNavigation_0, visibility=Collapsed}
   slot {class=SizeBoxSlot}
  }
  PanelsSlottingHost {
   widgetClassPath=/Script/WidgetSlotting.WidgetSlottingHostOverlay
   bIsVariable=true
   parent=Sidebar
   widget=PanelsSlottingHost
   properties {navigation=&self:WidgetTree.PanelsSlottingHost.WidgetNavigation_0, hostTag {tagName=UI.Sidebar.SubPanel}}
   slot {class=GridSlot, properties {padding=<ltrb 0 90 0 0>}}
  }
  OverlaySlottingHost {
   widgetClassPath=&WidgetSlottingHostOverlay
   bIsVariable=true
   parent=Sidebar
   widget=OverlaySlottingHost
   properties {hostTag {tagName=UI.Sidebar.Overlay}}
   slot {class=GridSlot, properties {layer=5}}
  }
 ]
}
